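# Post-copy animation Blueprint for the VealCadet male medium body cosmetic, driving secondary motion on its jacket, belt and collar.
# It copies the attached parent's pose, then weights jacket master bones across pelvis, spine, clavicle and pectoral bones with Constraint nodes.
# A Control Rig fed the emote flag follows, then angular-spring AnimDynamics with per-bone limits.
# The event graph caches the jacketFrontAlpha curve, which the AnimGraph never reads.
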AnimBlueprint /BRCosmetics/Characters/Player/Male/Medium/Bodies/M_MED_VealCadet/Meshes/M_MED_VealCadet_AnimBP
target_skeleton=&/M_MED_VealCadet_Skeleton
graphs [
 AnimGraph {nodes [
   n1 = "Constraint - dyn_jacket_l_inner_master (PPPPP)"!(
    ComponentPose=Local To Component(LocalPose=Copy Pose From Mesh() {bUseAttachedParent=true})
    Alpha=1
    ConstraintWeights_0=0.1878
    ConstraintWeights_1:float=0.2094
    ConstraintWeights_2:float=0.3283
    ConstraintWeights_3:float=0.2413
    ConstraintWeights_4:float=0.0333
   ) -> (Pose) {
    boneToModify {boneName=dyn_jacket_l_inner_master}
    constraintSetup [
     {targetBone {boneName=pelvis}}
     {targetBone {boneName=spine_01}}
     {targetBone {boneName=spine_02}}
     {targetBone {boneName=spine_03}}
     {targetBone {boneName=spine_04}}
    ]
    constraintWeights [1,1,1,1,1]
   }
   n2 = "Constraint - dyn_jacket_l_outer_fr_master (PPPPPP)"!(
    ComponentPose=@n1
    Alpha=1
    ConstraintWeights_0=0.1659
    ConstraintWeights_1:float=0.2864
    ConstraintWeights_2:float=0.1461
    ConstraintWeights_3:float=0.3919
    ConstraintWeights_4:float=0.0082
    ConstraintWeights_5:float=0.0014
   ) -> (Pose) {
    boneToModify {boneName=dyn_jacket_l_outer_fr_master}
    constraintSetup [
     {targetBone {boneName=pelvis}}
     {targetBone {boneName=spine_01}}
     {targetBone {boneName=spine_02}}
     {targetBone {boneName=spine_03}}
     {targetBone {boneName=clavicle_l}}
     {targetBone {boneName=deform_pec_l}}
    ]
    constraintWeights [1,1,1,1,1,1]
   }
   n3 = "Constraint - dyn_jacket_l_outer_side_master (PPPPPP)"!(
    ComponentPose=@n2
    Alpha=1
    ConstraintWeights_0=0.1433
    ConstraintWeights_1:float=0.21
    ConstraintWeights_2:float=0.2923
    ConstraintWeights_3:float=0.2961
    ConstraintWeights_4:float=0.0432
    ConstraintWeights_5:float=0.0152
   ) -> (Pose) {
    boneToModify {boneName=dyn_jacket_l_outer_side_master}
    constraintSetup [
     {targetBone {boneName=pelvis}}
     {targetBone {boneName=spine_01}}
     {targetBone {boneName=spine_02}}
     {targetBone {boneName=spine_03}}
     {targetBone {boneName=spine_04}}
     {targetBone {boneName=clavicle_l}}
    ]
    constraintWeights [1,1,1,1,1,1]
   }
   n4 = "Constraint - dyn_jacket_l_outer_bk_master (PPPPPP)"!(
    ComponentPose=@n3
    Alpha=1
    ConstraintWeights_0=0.1301
    ConstraintWeights_1:float=0.1842
    ConstraintWeights_2:float=0.235
    ConstraintWeights_3:float=0.4284
    ConstraintWeights_4:float=0.0002
    ConstraintWeights_5:float=0.0222
   ) -> (Pose) {
    boneToModify {boneName=dyn_jacket_l_outer_bk_master}
    constraintSetup [
     {targetBone {boneName=pelvis}}
     {targetBone {boneName=spine_01}}
     {targetBone {boneName=spine_02}}
     {targetBone {boneName=spine_03}}
     {targetBone {boneName=spine_05}}
     {targetBone {boneName=clavicle_l}}
    ]
    constraintWeights [1,1,1,1,1,1]
   }
   n5 = "Constraint - dyn_jacket_r_outer_bk_master (PPPPPP)"!(
    ComponentPose=@n4
    Alpha=1
    ConstraintWeights_0=0.1328
    ConstraintWeights_1:float=0.2099
    ConstraintWeights_2:float=0.2564
    ConstraintWeights_3:float=0.3785
    ConstraintWeights_4:float=0.0001
    ConstraintWeights_5:float=0.0224
   ) -> (Pose) {
    boneToModify {boneName=dyn_jacket_r_outer_bk_master}
    constraintSetup [
     {targetBone {boneName=pelvis}}
     {targetBone {boneName=spine_01}}
     {targetBone {boneName=spine_02}}
     {targetBone {boneName=spine_03}}
     {targetBone {boneName=spine_05}}
     {targetBone {boneName=clavicle_r}}
    ]
    constraintWeights [1,1,1,1,1,1]
   }
   n6 = "Constraint - dyn_jacket_r_outer_side_master (PPPPPP)"!(
    ComponentPose=@n5
    Alpha=1
    ConstraintWeights_0=0.1433
    ConstraintWeights_1:float=0.2099
    ConstraintWeights_2:float=0.2923
    ConstraintWeights_3:float=0.2961
    ConstraintWeights_4:float=0.0432
    ConstraintWeights_5:float=0.0152
   ) -> (Pose) {
    boneToModify {boneName=dyn_jacket_r_outer_side_master}
    constraintSetup [
     {targetBone {boneName=pelvis}}
     {targetBone {boneName=spine_01}}
     {targetBone {boneName=spine_02}}
     {targetBone {boneName=spine_03}}
     {targetBone {boneName=spine_04}}
     {targetBone {boneName=clavicle_r}}
    ]
    constraintWeights [1,1,1,1,1,1]
   }
   n7 = "Constraint - dyn_jacket_r_outer_fr_master (PPPPPP)"!(
    ComponentPose=@n6
    Alpha=1
    ConstraintWeights_0=0.1617
    ConstraintWeights_1:float=0.2817
    ConstraintWeights_2:float=0.1608
    ConstraintWeights_3:float=0.3839
    ConstraintWeights_4:float=0.0096
    ConstraintWeights_5:float=0.0024
   ) -> (Pose) {
    boneToModify {boneName=dyn_jacket_r_outer_fr_master}
    constraintSetup [
     {targetBone {boneName=pelvis}}
     {targetBone {boneName=spine_01}}
     {targetBone {boneName=spine_02}}
     {targetBone {boneName=spine_03}}
     {targetBone {boneName=clavicle_r}}
     {targetBone {boneName=deform_pec_r}}
    ]
    constraintWeights [1,1,1,1,1,1]
   }
   n8 = "Constraint - dyn_jacket_r_inner_master (PPPPP)"!(
    ComponentPose=@n7
    Alpha=1
    ConstraintWeights_0=0.1878
    ConstraintWeights_1:float=0.2094
    ConstraintWeights_2:float=0.3283
    ConstraintWeights_3:float=0.2413
    ConstraintWeights_4:float=0.0333
   ) -> (Pose) {
    boneToModify {boneName=dyn_jacket_r_inner_master}
    constraintSetup [
     {targetBone {boneName=pelvis}}
     {targetBone {boneName=spine_01}}
     {targetBone {boneName=spine_02}}
     {targetBone {boneName=spine_03}}
     {targetBone {boneName=spine_04}}
    ]
    constraintWeights [1,1,1,1,1]
   }
   n9 = Control Rig!(Source=Component To Local(ComponentPose=@n8), Alpha, isPlayingEmote:bool=Get bIsPlayingEmote()) -> (Pose) {controlRigAssetReference {controlRigAsset=&/M_MED_VealCadet_CtrlRig}}
   n10 = Anim Dynamics - Bone: dyn_belt_01(ComponentPose=Local To Component(LocalPose=@n9), Alpha=1, ExternalForce="0,0,0") {
    angularDampingOverride=1
    angularSpringConstant=500
    bAngularSpring=true
    bOverrideAngularDamping=true
    bUseGravityOverride=true
    simulationSpace=BoneRelative
    boundBone {boneName=dyn_belt_01}
    componentAppliedLinearAccClamp=<10000 10000 10000>
    componentLinearAccScale=<1 1 1>
    componentLinearVelScale=<1 1 1>
    gravityOverride {z=-980}
    physicsBodyDefinitions [{
      boundBone {boneName=dyn_belt_01}
      boxExtents=<10 10 10>
      constraintSetup {angularLimitsMax=<0 0 25>, angularLimitsMin=<0 0 0>, angularTarget=<1 0 0>}
      localJointOffset=<0.5 0 0>
     }]
    relativeSpaceBone {boneName=root}
    simSpaceSettings {maxAngularAcceleration=350, maxAngularVelocity=50, simSpaceAngularAlpha=1}
   }
   n11 = Anim Dynamics - Bone: dyn_jacket_l_inner_master(ComponentPose=@n10, Alpha=1, ExternalForce="0,0,0") {
    angularDampingOverride=1
    angularSpringConstant=500
    bAngularSpring=true
    bOverrideAngularDamping=true
    bUseGravityOverride=true
    simulationSpace=BoneRelative
    boundBone {boneName=dyn_jacket_l_inner_master}
    componentAppliedLinearAccClamp=<10000 10000 10000>
    componentLinearAccScale=<1 1 1>
    componentLinearVelScale=<1 1 1>
    gravityOverride {z=-980}
    physicsBodyDefinitions [{
      boundBone {boneName=dyn_jacket_l_inner_master}
      boxExtents=<10 10 10>
      constraintSetup {angularLimitsMax=<3 10 10>, angularLimitsMin=<-3 0 -10>, angularTarget=<1 0 0>}
      localJointOffset=<0.5 0 0>
     }]
    relativeSpaceBone {boneName=root}
    simSpaceSettings {maxAngularAcceleration=350, maxAngularVelocity=50, simSpaceAngularAlpha=1}
   }
   n12 = Anim Dynamics - Bone: dyn_jacket_l_outer_fr_master(ComponentPose=@n11, Alpha=1, ExternalForce="0,0,0") {
    angularDampingOverride=1
    angularSpringConstant=500
    bAngularSpring=true
    bOverrideAngularDamping=true
    bUseGravityOverride=true
    simulationSpace=BoneRelative
    boundBone {boneName=dyn_jacket_l_outer_fr_master}
    componentAppliedLinearAccClamp=<10000 10000 10000>
    componentLinearAccScale=<1 1 1>
    componentLinearVelScale=<1 1 1>
    gravityOverride {z=-980}
    physicsBodyDefinitions [{
      boundBone {boneName=dyn_jacket_l_outer_fr_master}
      boxExtents=<10 10 10>
      constraintSetup {angularLimitsMax=<3 10 35>, angularLimitsMin=<-3 -10 0>, angularTarget=<1 0 0>}
      localJointOffset=<0.5 0 0>
     }]
    relativeSpaceBone {boneName=root}
    simSpaceSettings {maxAngularAcceleration=350, maxAngularVelocity=50, simSpaceAngularAlpha=1}
   }
   n13 = Anim Dynamics - Bone: dyn_jacket_l_outer_side_master(ComponentPose=@n12, Alpha=1, ExternalForce="0,0,0") {
    angularDampingOverride=1
    angularSpringConstant=500
    bAngularSpring=true
    bOverrideAngularDamping=true
    bUseGravityOverride=true
    simulationSpace=BoneRelative
    boundBone {boneName=dyn_jacket_l_outer_side_master}
    componentAppliedLinearAccClamp=<10000 10000 10000>
    componentLinearAccScale=<1 1 1>
    componentLinearVelScale=<1 1 1>
    gravityOverride {z=-980}
    physicsBodyDefinitions [{
      boundBone {boneName=dyn_jacket_l_outer_side_master}
      boxExtents=<10 10 10>
      constraintSetup {angularLimitsMax=<3 0 10>, angularLimitsMin=<-3 -35 -10>, angularTarget=<1 0 0>}
      localJointOffset=<0.5 0 0>
     }]
    relativeSpaceBone {boneName=root}
    simSpaceSettings {maxAngularAcceleration=350, maxAngularVelocity=50, simSpaceAngularAlpha=1}
   }
   n14 = Anim Dynamics - Bone: dyn_jacket_l_outer_bk_master(ComponentPose=@n13, Alpha=1, ExternalForce="0,0,0") {
    angularDampingOverride=1
    angularSpringConstant=500
    bAngularSpring=true
    bOverrideAngularDamping=true
    bUseGravityOverride=true
    simulationSpace=BoneRelative
    boundBone {boneName=dyn_jacket_l_outer_bk_master}
    componentAppliedLinearAccClamp=<10000 10000 10000>
    componentLinearAccScale=<1 1 1>
    componentLinearVelScale=<1 1 1>
    gravityOverride {z=-980}
    physicsBodyDefinitions [{
      boundBone {boneName=dyn_jacket_l_outer_bk_master}
      boxExtents=<10 10 10>
      constraintSetup {angularLimitsMax=<3 0 10>, angularLimitsMin=<-3 -35 -10>, angularTarget=<1 0 0>}
      localJointOffset=<0.5 0 0>
     }]
    relativeSpaceBone {boneName=root}
    simSpaceSettings {maxAngularAcceleration=350, maxAngularVelocity=50, simSpaceAngularAlpha=1}
   }
   n15 = Anim Dynamics - Bone: dyn_jacket_r_outer_bk_master(ComponentPose=@n14, Alpha=1, ExternalForce="0,0,0") {
    angularDampingOverride=1
    angularSpringConstant=500
    bAngularSpring=true
    bOverrideAngularDamping=true
    bUseGravityOverride=true
    simulationSpace=BoneRelative
    boundBone {boneName=dyn_jacket_r_outer_bk_master}
    componentAppliedLinearAccClamp=<10000 10000 10000>
    componentLinearAccScale=<1 1 1>
    componentLinearVelScale=<1 1 1>
    gravityOverride {z=-980}
    physicsBodyDefinitions [{
      boundBone {boneName=dyn_jacket_r_outer_bk_master}
      boxExtents=<10 10 10>
      constraintSetup {angularLimitsMax=<3 0 10>, angularLimitsMin=<-3 -35 -10>, angularTarget=<1 0 0>}
      localJointOffset=<0.5 0 0>
     }]
    relativeSpaceBone {boneName=root}
    simSpaceSettings {maxAngularAcceleration=350, maxAngularVelocity=50, simSpaceAngularAlpha=1}
   }
   n16 = Anim Dynamics - Bone: dyn_jacket_r_outer_side_master(ComponentPose=@n15, Alpha=1, ExternalForce="0,0,0") {
    angularDampingOverride=1
    angularSpringConstant=500
    bAngularSpring=true
    bOverrideAngularDamping=true
    bUseGravityOverride=true
    simulationSpace=BoneRelative
    boundBone {boneName=dyn_jacket_r_outer_side_master}
    componentAppliedLinearAccClamp=<10000 10000 10000>
    componentLinearAccScale=<1 1 1>
    componentLinearVelScale=<1 1 1>
    gravityOverride {z=-980}
    physicsBodyDefinitions [{
      boundBone {boneName=dyn_jacket_r_outer_side_master}
      boxExtents=<10 10 10>
      constraintSetup {angularLimitsMax=<3 35 10>, angularLimitsMin=<-3 0 -10>, angularTarget=<1 0 0>}
      localJointOffset=<0.5 0 0>
     }]
    relativeSpaceBone {boneName=root}
    simSpaceSettings {maxAngularAcceleration=350, maxAngularVelocity=50, simSpaceAngularAlpha=1}
   }
   n17 = Anim Dynamics - Bone: dyn_jacket_r_outer_fr_master(ComponentPose=@n16, Alpha=1, ExternalForce="0,0,0") {
    angularDampingOverride=1
    angularSpringConstant=500
    bAngularSpring=true
    bOverrideAngularDamping=true
    bUseGravityOverride=true
    simulationSpace=BoneRelative
    boundBone {boneName=dyn_jacket_r_outer_fr_master}
    componentAppliedLinearAccClamp=<10000 10000 10000>
    componentLinearAccScale=<1 1 1>
    componentLinearVelScale=<1 1 1>
    gravityOverride {z=-980}
    physicsBodyDefinitions [{
      boundBone {boneName=dyn_jacket_r_outer_fr_master}
      boxExtents=<10 10 10>
      constraintSetup {angularLimitsMax=<3 35 10>, angularLimitsMin=<-3 0 -10>, angularTarget=<1 0 0>}
      localJointOffset=<0.5 0 0>
     }]
    relativeSpaceBone {boneName=root}
    simSpaceSettings {maxAngularAcceleration=350, maxAngularVelocity=50, simSpaceAngularAlpha=1}
   }
   n18 = Anim Dynamics - Bone: dyn_jacket_r_inner_master(ComponentPose=@n17, Alpha=1, ExternalForce="0,0,0") {
    angularDampingOverride=1
    angularSpringConstant=500
    bAngularSpring=true
    bOverrideAngularDamping=true
    bUseGravityOverride=true
    simulationSpace=BoneRelative
    boundBone {boneName=dyn_jacket_r_inner_master}
    componentAppliedLinearAccClamp=<10000 10000 10000>
    componentLinearAccScale=<1 1 1>
    componentLinearVelScale=<1 1 1>
    gravityOverride {z=-980}
    physicsBodyDefinitions [{
      boundBone {boneName=dyn_jacket_r_inner_master}
      boxExtents=<10 10 10>
      constraintSetup {angularLimitsMax=<3 10 10>, angularLimitsMin=<-3 0 -10>, angularTarget=<1 0 0>}
      localJointOffset=<0.5 0 0>
     }]
    relativeSpaceBone {boneName=root}
    simSpaceSettings {maxAngularAcceleration=350, maxAngularVelocity=50, simSpaceAngularAlpha=1}
   }
   n19 = Anim Dynamics - Bone: dyn_collar_l_lower_01(ComponentPose=@n18, Alpha=1, ExternalForce="0,0,0") {
    angularDampingOverride=1
    angularSpringConstant=800
    bAngularSpring=true
    bOverrideAngularDamping=true
    bUseGravityOverride=true
    simulationSpace=BoneRelative
    boundBone {boneName=dyn_collar_l_lower_01}
    componentAppliedLinearAccClamp=<5000 5000 5000>
    componentLinearAccScale=<1 1 1>
    componentLinearVelScale=<1 1 1>
    gravityOverride {z=-980}
    physicsBodyDefinitions [{
      boundBone {boneName=dyn_collar_l_lower_01}
      boxExtents=<10 10 10>
      constraintSetup {angularLimitsMax=<3 10 0>, angularLimitsMin=<-3 -10 0>, angularTarget=<1 0 0>}
      localJointOffset=<0.5 0 0>
     }]
    relativeSpaceBone {boneName=root}
    simSpaceSettings {maxAngularAcceleration=350, maxAngularVelocity=50, simSpaceAngularAlpha=1}
   }
   n20 = Anim Dynamics - Bone: dyn_collar_r_lower_01(ComponentPose=@n19, Alpha=1, ExternalForce="0,0,0") {
    angularDampingOverride=1
    angularSpringConstant=800
    bAngularSpring=true
    bOverrideAngularDamping=true
    bUseGravityOverride=true
    simulationSpace=BoneRelative
    boundBone {boneName=dyn_collar_r_lower_01}
    componentAppliedLinearAccClamp=<5000 5000 5000>
    componentLinearAccScale=<1 1 1>
    componentLinearVelScale=<1 1 1>
    gravityOverride {z=-980}
    physicsBodyDefinitions [{
      boundBone {boneName=dyn_collar_r_lower_01}
      boxExtents=<10 10 10>
      constraintSetup {angularLimitsMax=<3 10 0>, angularLimitsMin=<-3 -10 0>, angularTarget=<1 0 0>}
      localJointOffset=<0.5 0 0>
     }]
    relativeSpaceBone {boneName=root}
    simSpaceSettings {maxAngularAcceleration=350, maxAngularVelocity=50, simSpaceAngularAlpha=1}
   }
   n21 = Anim Dynamics - Bone: dyn_collar_l_mid_01(ComponentPose=@n20, Alpha=1, ExternalForce="0,0,0") {
    angularDampingOverride=1
    angularSpringConstant=800
    bAngularSpring=true
    bOverrideAngularDamping=true
    bUseGravityOverride=true
    simulationSpace=BoneRelative
    boundBone {boneName=dyn_collar_l_mid_01}
    componentAppliedLinearAccClamp=<5000 5000 5000>
    componentLinearAccScale=<1 1 1>
    componentLinearVelScale=<1 1 1>
    gravityOverride {z=-980}
    physicsBodyDefinitions [{
      boundBone {boneName=dyn_collar_l_mid_01}
      boxExtents=<10 10 10>
      constraintSetup {angularLimitsMax=<3 10 0>, angularLimitsMin=<-3 -10 0>, angularTarget=<1 0 0>}
      localJointOffset=<0.1 0 0>
     }]
    relativeSpaceBone {boneName=root}
    simSpaceSettings {maxAngularAcceleration=350, maxAngularVelocity=50, simSpaceAngularAlpha=1}
   }
   n22 = Anim Dynamics - Bone: dyn_collar_r_mid_01(ComponentPose=@n21, Alpha=1, ExternalForce="0,0,0") {
    angularDampingOverride=1
    angularSpringConstant=800
    bAngularSpring=true
    bOverrideAngularDamping=true
    bUseGravityOverride=true
    simulationSpace=BoneRelative
    boundBone {boneName=dyn_collar_r_mid_01}
    componentAppliedLinearAccClamp=<5000 5000 5000>
    componentLinearAccScale=<1 1 1>
    componentLinearVelScale=<1 1 1>
    gravityOverride {z=-980}
    physicsBodyDefinitions [{
      boundBone {boneName=dyn_collar_r_mid_01}
      boxExtents=<10 10 10>
      constraintSetup {angularLimitsMax=<3 10 0>, angularLimitsMin=<-3 -1 0>, angularTarget=<1 0 0>}
      localJointOffset=<-0.1 0 0>
     }]
    relativeSpaceBone {boneName=root}
    simSpaceSettings {maxAngularAcceleration=350, maxAngularVelocity=50, simSpaceAngularAlpha=1}
   }
   Output Pose(Result=Component To Local(ComponentPose=@n22)) {layerGroup=None}
  ]}
 EventGraph {nodes [
   Event BlueprintUpdateAnimation()
   > Set jacketFrontAlpha(jacketFrontAlpha:float=GetCurveValue(CurveName=jacketFrontAlpha)) -> (Output_Get:float)
  ]}
]
variables [jacketFrontAlpha {type=number}]
