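# Physics asset for the face accessory/backpack part of the male medium VealCadet cosmetic, covering left, middle and right dyn_hair bone chains.
# Capsules wrap those chains plus head, neck, upper spine, clavicles and upper arms.
# Left and right chains have tight swing and twist limits from master to tip.
# Each side's second bone also links to the middle's with free rotation; the middle chain has no constraints along its length.
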PhysicsAsset /BRCosmetics/Characters/Player/Male/Medium/Bodies/M_MED_VealCadet/Meshes/Parts/M_MED_VealCadet_FaceAcc_Backpack_Physics
bodies {
 clavicle_l [{
   shapeType=Capsule
   length=9.5
   radius=7
   shapeName=clavicle_l_capsule
   center=<5.9301 -0.5329 0.5762>
   rotation=<rot -0.2275 -90.0631 99.0714>
  }]
 clavicle_r [{
   shapeType=Capsule
   length=9.5
   radius=7
   shapeName=clavicle_r_capsule
   center=<-5.9301 0.5822 -0.5761>
   rotation=<rot -0.2275 -90.0631 -80.9286>
  }]
 dyn_hair_l_primary_01 [{shapeType=Capsule, length=5, radius=4.5, shapeName=dyn_hair_l_primary_01_capsule, center=<4.629 -4.8901 0>, rotation=<rot -9.9999 0.0001 0>}]
 dyn_hair_l_primary_02 [{
   shapeType=Capsule
   length=5.5
   radius=3.5
   shapeName=dyn_hair_l_primary_02_capsule
   center=<4.8651 -2.2159 -0.075>
   rotation=<rot 88.0224 8.406 -17.3893>
  }]
 dyn_hair_l_primary_03 [{
   shapeType=Capsule
   length=6
   radius=3.5
   shapeName=dyn_hair_l_primary_03_capsule
   center=<1.6722 0.3755 1.2519>
   rotation=<rot 83.1734 72.323 99.1685>
  }]
 dyn_hair_l_primary_04 [{
   shapeType=Capsule
   length=6
   radius=4
   shapeName=dyn_hair_l_primary_04_capsule
   center=<2.4531 0.0105 0.4714>
   rotation=<rot 60 -90 -90.0001>
  }]
 dyn_hair_l_primary_master [{shapeType=Capsule, length=0.505, radius=0.505, shapeName=dyn_hair_l_primary_master_capsule, rotation=<rot 0 -90 0>}]
 dyn_hair_m_primary_01 [{shapeType=Capsule, length=9, radius=5.5, shapeName=dyn_hair_m_primary_01_capsule, center=<8.643 -4.8171 0>, rotation=<rot 90 -3.3413 0>}]
 dyn_hair_m_primary_02 [{shapeType=Capsule, length=5, radius=4.5, shapeName=dyn_hair_m_primary_02_capsule, center=<4.6894 -3.0531 0>, rotation=<rot 90 17.3546 0>}]
 dyn_hair_m_primary_03 [{shapeType=Capsule, length=5, radius=4, shapeName=dyn_hair_m_primary_03_capsule, center=<2.5 0 0>, rotation=<rot 90 0 0>}]
 dyn_hair_m_primary_04 [{shapeType=Capsule, length=5.5, radius=6, shapeName=dyn_hair_m_primary_04_capsule, center=<4.3197 -0.0749 0>, rotation=<rot 0 -179.9999 180>}]
 dyn_hair_m_primary_master [{shapeType=Capsule, length=0.505, radius=0.505, shapeName=dyn_hair_m_primary_master_capsule, rotation=<rot 0 -90 0>}]
 dyn_hair_r_primary_01 [{
   shapeType=Capsule
   length=5
   radius=4
   shapeName=dyn_hair_r_primary_01_capsule
   center=<2.5886 -4.6966 -0.5027>
   rotation=<rot 9.9999 0.0001 0>
  }]
 dyn_hair_r_primary_02 [{
   shapeType=Capsule
   length=6
   radius=3.5
   shapeName=dyn_hair_r_primary_02_capsule
   center=<4.2662 -2.8983 -0.0511>
   rotation=<rot 88.022 8.4086 -9.5628>
  }]
 dyn_hair_r_primary_03 [{
   shapeType=Capsule
   length=4.5
   radius=3.5
   shapeName=dyn_hair_r_primary_03_capsule
   center=<4.4929 0.0643 -1.3409>
   rotation=<rot 50 88.5062 89.9999>
  }]
 dyn_hair_r_primary_04 [{
   shapeType=Capsule
   length=3
   radius=4
   shapeName=dyn_hair_r_primary_04_capsule
   center=<5.8512 -0.054 0.5262>
   rotation=<rot 79.7066 170.5873 170.2847>
  }]
 dyn_hair_r_primary_master [{shapeType=Capsule, length=0.505, radius=0.505, shapeName=dyn_hair_r_primary_master_capsule, rotation=<rot 0 -90 0>}]
 head [{
   shapeType=Capsule
   length=10.2264
   radius=8.6383
   shapeName=head_capsule
   center=<3.7188 -0.6899 0.1915>
   rotation=<rot 0.4553 111.8896 -89.0567>
  }]
 neck_02 [{shapeType=Capsule, length=2.5, radius=5.5, shapeName=neck_02_capsule, center=<0 -2.4015 0>, rotation=<rot 0 -90.0001 90.2765>}]
 spine_05 [{shapeType=Capsule, length=11, radius=10, shapeName=spine_05_capsule, center=<1.3441 0.3898 0>, rotation=<rot 0 -90 0>}]
 upperarm_l [{
   shapeType=Capsule
   length=9.5
   radius=5.5
   shapeName=upperarm_l_capsule
   center=<4.5123 0.0001 -0.1433>
   rotation=<rot -0.0008 -89.9993 91.8189>
  }]
 upperarm_r [{
   shapeType=Capsule
   length=9.5
   radius=5.5
   shapeName=upperarm_r_capsule
   center=<-4.5123 -0.0001 0.1433>
   rotation=<rot -0.0008 -89.9993 -88.1811>
  }]
}
constraints [
 {bone1Name=dyn_hair_l_primary_02, bone2Name=dyn_hair_l_primary_01, swing1LimitDegrees=25, swing2LimitDegrees=25, twistLimitDegrees=10}
 {bone1Name=dyn_hair_r_primary_02, bone2Name=dyn_hair_r_primary_01, swing1LimitDegrees=25, swing2LimitDegrees=25, twistLimitDegrees=10}
 {bone1Name=dyn_hair_l_primary_03, bone2Name=dyn_hair_l_primary_02, swing1LimitDegrees=25, swing2LimitDegrees=25, twistLimitDegrees=10}
 {bone1Name=dyn_hair_l_primary_04, bone2Name=dyn_hair_l_primary_03, swing1LimitDegrees=25, swing2LimitDegrees=25, twistLimitDegrees=10}
 {bone1Name=dyn_hair_r_primary_03, bone2Name=dyn_hair_r_primary_02, swing1LimitDegrees=25, swing2LimitDegrees=25, twistLimitDegrees=10}
 {bone1Name=dyn_hair_r_primary_04, bone2Name=dyn_hair_r_primary_03, swing1LimitDegrees=25, swing2LimitDegrees=25, twistLimitDegrees=10}
 {bone1Name=dyn_hair_r_primary_01, bone2Name=dyn_hair_r_primary_master, swing1LimitDegrees=25, swing2LimitDegrees=25, twistLimitDegrees=10}
 {bone1Name=dyn_hair_l_primary_01, bone2Name=dyn_hair_l_primary_master, swing1LimitDegrees=25, swing2LimitDegrees=25, twistLimitDegrees=10}
 {
  bone1Name=dyn_hair_r_primary_02
  bone2Name=dyn_hair_m_primary_02
  swing1LimitDegrees=45
  swing1Motion=Free
  swing2LimitDegrees=45
  swing2Motion=Free
  twistLimitDegrees=45
  twistMotion=Free
 }
 {
  bone1Name=dyn_hair_l_primary_02
  bone2Name=dyn_hair_m_primary_02
  swing1LimitDegrees=45
  swing1Motion=Free
  swing2LimitDegrees=45
  swing2Motion=Free
  twistLimitDegrees=45
  twistMotion=Free
 }
]
