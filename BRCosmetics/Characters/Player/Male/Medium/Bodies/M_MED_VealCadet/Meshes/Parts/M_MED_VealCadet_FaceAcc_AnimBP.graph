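# Animation Blueprint for the Veal Cadet face accessory skeleton, adding secondary motion to the character's hair.
# Two RigidBody simulations run from a cached PreSim pose, with isWearingBackbling selecting between the backpack physics asset (head space) and the default one (neck_01 space).
# Weighted multi-target constraints then attach secondary hair master bones to the head and primary hair bones.
# Per-bone AnimDynamics nodes layer limited, sprung angular jiggle onto the secondary strands.
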AnimBlueprint /BRCosmetics/Characters/Player/Male/Medium/Bodies/M_MED_VealCadet/Meshes/Parts/M_MED_VealCadet_FaceAcc_AnimBP
target_skeleton=&/M_MED_VealCadet_FaceAcc_Skeleton
graphs [
 AnimGraph {nodes [
   n1 = RigidBody!(
    ComponentPose=Local To Component(LocalPose=Use cached pose 'PreSim'())
    bAlphaBoolEnabled:bool
    OverrideWorldGravity:Vector="0,0,0"
    ExternalForce
   ) -> (Pose) {
    alphaInputType=Bool
    bClampLinearTranslationLimitToRefPose=true
    bOverrideWorldGravity=true
    overridePhysicsAsset=&/M_MED_VealCadet_FaceAcc_Backpack_Physics
    simulationSpace=BaseBoneSpace
    baseBoneRef {boneName=head}
    componentAppliedLinearAccClamp=<0 0 0>
    simSpaceSettings {
     dampingAlpha=0.8
     maxAngularAcceleration=120
     maxAngularVelocity=20
     maxLinearAcceleration=400
     maxLinearVelocity=250
     worldAlpha=1
     externalLinearDragV=<0.8 0.8 0.8>
    }
   }
   n2 = Component To Local(ComponentPose=@n1)
   n3 = RigidBody!(ComponentPose=Local To Component(LocalPose=Use cached pose 'PreSim'()), Alpha:float, OverrideWorldGravity="0,0,0", ExternalForce) -> (Pose) {
    bClampLinearTranslationLimitToRefPose=true
    bOverrideWorldGravity=true
    overridePhysicsAsset=&/M_MED_VealCadet_FaceAcc_Physics
    simulationSpace=BaseBoneSpace
    baseBoneRef {boneName=neck_01}
    componentAppliedLinearAccClamp=<0 0 0>
    simSpaceSettings {
     dampingAlpha=0.25
     maxAngularAcceleration=120
     maxAngularVelocity=20
     maxLinearAcceleration=800
     maxLinearVelocity=500
     worldAlpha=1
     externalLinearDragV=<1 1 1>
    }
   }
   n4 = Component To Local(ComponentPose=@n3)
   n5 = Get isWearingBackbling()
   n6 = "Constraint - dyn_hair_scd_a_master (P)"!(
    ComponentPose=Local To Component(LocalPose=Blend Poses by bool(BlendPose_0=@n2, BlendPose_1=@n4, bActiveValue=@n5))
    Alpha=1
    ConstraintWeights_0=1
    ConstraintWeights_1:float=0
   ) -> (Pose) {boneToModify {boneName=dyn_hair_scd_a_master}, constraintSetup [{targetBone {boneName=head}}], constraintWeights [1,0]}
   n7 = "Constraint - dyn_hair_scd_b_master (P)"!(ComponentPose=@n6, Alpha=1, ConstraintWeights_0=1, ConstraintWeights_1:float=0) -> (Pose) {boneToModify {boneName=dyn_hair_scd_b_master}, constraintSetup [{targetBone {boneName=head}}], constraintWeights [1,0]}
   n8 = "Constraint - dyn_hair_scd_c_01 (P)"!(ComponentPose=@n7, Alpha=1, ConstraintWeights_0=1, ConstraintWeights_1:float=0) -> (Pose) {boneToModify {boneName=dyn_hair_scd_c_01}, constraintSetup [{targetBone {boneName=head}}], constraintWeights [1,0]}
   n9 = "Constraint - dyn_hair_scd_d_master (PPP)"!(ComponentPose=@n8, Alpha=1, ConstraintWeights_0=1, ConstraintWeights_1:float=0, ConstraintWeights_2:float=0) -> (Pose) {
    boneToModify {boneName=dyn_hair_scd_d_master}
    constraintSetup [{targetBone {boneName=head}}, {targetBone {boneName=dyn_hair_l_primary_01}}, {targetBone {boneName=dyn_hair_l_primary_02}}]
    constraintWeights [1,1,1]
   }
   n10 = "Constraint - dyn_hair_scd_e_master (PPPPPPPP)"!(
    ComponentPose=@n9
    Alpha=1
    ConstraintWeights_0=0.9834
    ConstraintWeights_1:float=0.0155
    ConstraintWeights_2:float=0
    ConstraintWeights_3:float=0
    ConstraintWeights_4:float=0
    ConstraintWeights_5:float=0
    ConstraintWeights_6:float=0.0011
    ConstraintWeights_7:float=0
   ) -> (Pose) {
    boneToModify {boneName=dyn_hair_scd_e_master}
    constraintSetup [
     {targetBone {boneName=head}}
     {targetBone {boneName=dyn_hair_l_primary_01}}
     {targetBone {boneName=dyn_hair_l_primary_03}}
     {targetBone {boneName=dyn_hair_m_primary_01}}
     {targetBone {boneName=dyn_hair_m_primary_02}}
     {targetBone {boneName=dyn_hair_r_primary_02}}
     {targetBone {boneName=dyn_hair_l_primary_02}}
     {targetBone {boneName=dyn_hair_r_primary_01}}
    ]
    constraintWeights [1,1,1,1,1,1,1,1]
   }
   n11 = "Constraint - dyn_hair_scd_f_master (PPPPPPPPP)"!(
    ComponentPose=@n10
    Alpha=1
    ConstraintWeights_0=0.9823
    ConstraintWeights_1:float=0
    ConstraintWeights_2:float=0.0001
    ConstraintWeights_3:float=0
    ConstraintWeights_4:float=0
    ConstraintWeights_5:float=0.0058
    ConstraintWeights_6:float=0.0004
    ConstraintWeights_7:float=0
    ConstraintWeights_8:float=0.0114
   ) -> (Pose) {
    boneToModify {boneName=dyn_hair_scd_f_master}
    constraintSetup [
     {targetBone {boneName=head}}
     {targetBone {boneName=dyn_hair_l_primary_01}}
     {targetBone {boneName=dyn_hair_m_primary_01}}
     {targetBone {boneName=dyn_hair_m_primary_02}}
     {targetBone {boneName=dyn_hair_m_primary_03}}
     {targetBone {boneName=dyn_hair_r_primary_02}}
     {targetBone {boneName=dyn_hair_r_primary_03}}
     {targetBone {boneName=dyn_hair_l_primary_02}}
     {targetBone {boneName=dyn_hair_r_primary_01}}
    ]
    constraintWeights [1,1,1,1,1,1,1,1,1]
   }
   n12 = "Constraint - dyn_hair_scd_g_master (PPPPPPPPP)"!(
    ComponentPose=@n11
    Alpha=1
    ConstraintWeights_0=0
    ConstraintWeights_1:float=0.0031
    ConstraintWeights_2:float=0.3731
    ConstraintWeights_3:float=0
    ConstraintWeights_4:float=0
    ConstraintWeights_5:float=0
    ConstraintWeights_6:float=0.6238
    ConstraintWeights_7:float=0
    ConstraintWeights_8:float=0
   ) -> (Pose) {
    boneToModify {boneName=dyn_hair_scd_g_master}
    constraintSetup [
     {targetBone {boneName=head}}
     {targetBone {boneName=dyn_hair_l_primary_01}}
     {targetBone {boneName=dyn_hair_l_primary_03}}
     {targetBone {boneName=dyn_hair_m_primary_01}}
     {targetBone {boneName=dyn_hair_m_primary_02}}
     {targetBone {boneName=dyn_hair_m_primary_03}}
     {targetBone {boneName=dyn_hair_l_primary_02}}
     {targetBone {boneName=dyn_hair_l_primary_04}}
     {targetBone {boneName=dyn_hair_m_primary_04}}
    ]
    constraintWeights [1,1,1,1,1,1,1,1,1]
   }
   n13 = "Constraint - dyn_hair_scd_h_master (PPPPPPPP)"!(
    ComponentPose=@n12
    Alpha=1
    ConstraintWeights_0=0.0219
    ConstraintWeights_1:float=0.289
    ConstraintWeights_2:float=0.0099
    ConstraintWeights_3:float=0.0002
    ConstraintWeights_4:float=0.0001
    ConstraintWeights_5:float=0
    ConstraintWeights_6:float=0.6789
    ConstraintWeights_7:float=0
   ) -> (Pose) {
    boneToModify {boneName=dyn_hair_scd_h_master}
    constraintSetup [
     {targetBone {boneName=head}}
     {targetBone {boneName=dyn_hair_l_primary_01}}
     {targetBone {boneName=dyn_hair_l_primary_03}}
     {targetBone {boneName=dyn_hair_m_primary_01}}
     {targetBone {boneName=dyn_hair_m_primary_02}}
     {targetBone {boneName=dyn_hair_r_primary_02}}
     {targetBone {boneName=dyn_hair_l_primary_02}}
     {targetBone {boneName=dyn_hair_r_primary_01}}
    ]
    constraintWeights [1,1,1,1,1,1,1,1]
   }
   n14 = "Constraint - dyn_hair_scd_i_master (PPPPPPPP)"!(
    ComponentPose=@n13
    Alpha=1
    ConstraintWeights_0=0.0265
    ConstraintWeights_1:float=0.3834
    ConstraintWeights_2:float=0.0003
    ConstraintWeights_3:float=0.0051
    ConstraintWeights_4:float=0.004
    ConstraintWeights_5:float=0.0001
    ConstraintWeights_6:float=0.5802
    ConstraintWeights_7:float=0.0004
   ) -> (Pose) {
    boneToModify {boneName=dyn_hair_scd_i_master}
    constraintSetup [
     {targetBone {boneName=head}}
     {targetBone {boneName=dyn_hair_l_primary_01}}
     {targetBone {boneName=dyn_hair_l_primary_03}}
     {targetBone {boneName=dyn_hair_m_primary_01}}
     {targetBone {boneName=dyn_hair_m_primary_02}}
     {targetBone {boneName=dyn_hair_r_primary_02}}
     {targetBone {boneName=dyn_hair_l_primary_02}}
     {targetBone {boneName=dyn_hair_r_primary_01}}
    ]
    constraintWeights [1,1,1,1,1,1,1,1]
   }
   n15 = "Constraint - dyn_hair_scd_j_master (PPPPPPPP)"!(
    ComponentPose=@n14
    Alpha=1
    ConstraintWeights_0=0.003
    ConstraintWeights_1:float=0.1507
    ConstraintWeights_2:float=0.0271
    ConstraintWeights_3:float=0.008
    ConstraintWeights_4:float=0.0143
    ConstraintWeights_5:float=0.0001
    ConstraintWeights_6:float=0.7966
    ConstraintWeights_7:float=0.0003
   ) -> (Pose) {
    boneToModify {boneName=dyn_hair_scd_j_master}
    constraintSetup [
     {targetBone {boneName=head}}
     {targetBone {boneName=dyn_hair_l_primary_01}}
     {targetBone {boneName=dyn_hair_l_primary_03}}
     {targetBone {boneName=dyn_hair_m_primary_01}}
     {targetBone {boneName=dyn_hair_m_primary_02}}
     {targetBone {boneName=dyn_hair_r_primary_02}}
     {targetBone {boneName=dyn_hair_l_primary_02}}
     {targetBone {boneName=dyn_hair_r_primary_01}}
    ]
    constraintWeights [1,1,1,1,1,1,1,1]
   }
   n16 = "Constraint - dyn_hair_scd_k_master (PPPPPPPP)"!(
    ComponentPose=@n15
    Alpha=1
    ConstraintWeights_0=0.0003
    ConstraintWeights_1:float=0.0124
    ConstraintWeights_2:float=0.2671
    ConstraintWeights_3:float=0.0053
    ConstraintWeights_4:float=0.0835
    ConstraintWeights_5:float=0.0246
    ConstraintWeights_6:float=0.6067
    ConstraintWeights_7:float=0.0001
   ) -> (Pose) {
    boneToModify {boneName=dyn_hair_scd_k_master}
    constraintSetup [
     {targetBone {boneName=head}}
     {targetBone {boneName=dyn_hair_l_primary_01}}
     {targetBone {boneName=dyn_hair_l_primary_03}}
     {targetBone {boneName=dyn_hair_m_primary_01}}
     {targetBone {boneName=dyn_hair_m_primary_02}}
     {targetBone {boneName=dyn_hair_m_primary_03}}
     {targetBone {boneName=dyn_hair_l_primary_02}}
     {targetBone {boneName=dyn_hair_r_primary_01}}
    ]
    constraintWeights [1,1,1,1,1,1,1,1]
   }
   n17 = "Constraint - dyn_hair_scd_l_master (PPPPPPPP)"!(
    ComponentPose=@n16
    Alpha=1
    ConstraintWeights_0=0
    ConstraintWeights_1:float=0.7877
    ConstraintWeights_2:float=0
    ConstraintWeights_3:float=0.0002
    ConstraintWeights_4:float=0.0003
    ConstraintWeights_5:float=0.2101
    ConstraintWeights_6:float=0.0018
    ConstraintWeights_7:float=0
   ) -> (Pose) {
    boneToModify {boneName=dyn_hair_scd_l_master}
    constraintSetup [
     {targetBone {boneName=dyn_hair_l_primary_01}}
     {targetBone {boneName=dyn_hair_l_primary_03}}
     {targetBone {boneName=dyn_hair_m_primary_01}}
     {targetBone {boneName=dyn_hair_m_primary_02}}
     {targetBone {boneName=dyn_hair_m_primary_03}}
     {targetBone {boneName=dyn_hair_l_primary_02}}
     {targetBone {boneName=dyn_hair_l_primary_04}}
     {targetBone {boneName=dyn_hair_m_primary_04}}
    ]
    constraintWeights [1,1,1,1,1,1,1,1]
   }
   n18 = "Constraint - dyn_hair_scd_m_master (PPPPPPPPP)"!(
    ComponentPose=@n17
    Alpha=1
    ConstraintWeights_0=0.0028
    ConstraintWeights_1:float=0.2279
    ConstraintWeights_2:float=0.004
    ConstraintWeights_3:float=0.2923
    ConstraintWeights_4:float=0.184
    ConstraintWeights_5:float=0.0001
    ConstraintWeights_6:float=0.2888
    ConstraintWeights_7:float=0.0002
    ConstraintWeights_8:float=0
   ) -> (Pose) {
    boneToModify {boneName=dyn_hair_scd_m_master}
    constraintSetup [
     {targetBone {boneName=dyn_hair_l_primary_01}}
     {targetBone {boneName=dyn_hair_l_primary_03}}
     {targetBone {boneName=dyn_hair_m_primary_01}}
     {targetBone {boneName=dyn_hair_m_primary_02}}
     {targetBone {boneName=dyn_hair_m_primary_03}}
     {targetBone {boneName=dyn_hair_r_primary_02}}
     {targetBone {boneName=dyn_hair_l_primary_02}}
     {targetBone {boneName=dyn_hair_l_primary_04}}
     {targetBone {boneName=dyn_hair_m_primary_04}}
    ]
    constraintWeights [1,1,1,1,1,1,1,1,1]
   }
   n19 = "Constraint - dyn_hair_scd_n_master (PPPPPPPP)"!(
    ComponentPose=@n18
    Alpha=1
    ConstraintWeights_0=0.0023
    ConstraintWeights_1:float=0.0353
    ConstraintWeights_2:float=0.0213
    ConstraintWeights_3:float=0.653
    ConstraintWeights_4:float=0.1315
    ConstraintWeights_5:float=0.0106
    ConstraintWeights_6:float=0.0046
    ConstraintWeights_7:float=0.1413
   ) -> (Pose) {
    boneToModify {boneName=dyn_hair_scd_n_master}
    constraintSetup [
     {targetBone {boneName=dyn_hair_l_primary_01}}
     {targetBone {boneName=dyn_hair_l_primary_03}}
     {targetBone {boneName=dyn_hair_m_primary_01}}
     {targetBone {boneName=dyn_hair_m_primary_02}}
     {targetBone {boneName=dyn_hair_m_primary_03}}
     {targetBone {boneName=dyn_hair_r_primary_02}}
     {targetBone {boneName=dyn_hair_r_primary_03}}
     {targetBone {boneName=dyn_hair_l_primary_02}}
    ]
    constraintWeights [1,1,1,1,1,1,1,1]
   }
   n20 = "Constraint - dyn_hair_scd_o_master (PPPPPPPPP)"!(
    ComponentPose=@n19
    Alpha=1
    ConstraintWeights_0=0.2286
    ConstraintWeights_1:float=0.0004
    ConstraintWeights_2:float=0.2202
    ConstraintWeights_3:float=0.4586
    ConstraintWeights_4:float=0.0007
    ConstraintWeights_5:float=0.0006
    ConstraintWeights_6:float=0.0902
    ConstraintWeights_7:float=0.0007
    ConstraintWeights_8:float=0
   ) -> (Pose) {
    boneToModify {boneName=dyn_hair_scd_o_master}
    constraintSetup [
     {targetBone {boneName=dyn_hair_l_primary_03}}
     {targetBone {boneName=dyn_hair_m_primary_01}}
     {targetBone {boneName=dyn_hair_m_primary_02}}
     {targetBone {boneName=dyn_hair_m_primary_03}}
     {targetBone {boneName=dyn_hair_r_primary_02}}
     {targetBone {boneName=dyn_hair_r_primary_03}}
     {targetBone {boneName=dyn_hair_l_primary_02}}
     {targetBone {boneName=dyn_hair_l_primary_04}}
     {targetBone {boneName=dyn_hair_r_primary_04}}
    ]
    constraintWeights [1,1,1,1,1,1,1,1,1]
   }
   n21 = "Constraint - dyn_hair_scd_p_master (PPPPPPPPPP)"!(
    ComponentPose=@n20
    Alpha=1
    ConstraintWeights_0=0.0003
    ConstraintWeights_1:float=0
    ConstraintWeights_2:float=0.289
    ConstraintWeights_3:float=0.6836
    ConstraintWeights_4:float=0.0072
    ConstraintWeights_5:float=0.0179
    ConstraintWeights_6:float=0.0002
    ConstraintWeights_7:float=0.001
    ConstraintWeights_8:float=0
    ConstraintWeights_9:float=0.0008
   ) -> (Pose) {
    boneToModify {boneName=dyn_hair_scd_p_master}
    constraintSetup [
     {targetBone {boneName=dyn_hair_l_primary_03}}
     {targetBone {boneName=dyn_hair_m_primary_01}}
     {targetBone {boneName=dyn_hair_m_primary_02}}
     {targetBone {boneName=dyn_hair_m_primary_03}}
     {targetBone {boneName=dyn_hair_r_primary_02}}
     {targetBone {boneName=dyn_hair_r_primary_03}}
     {targetBone {boneName=dyn_hair_l_primary_02}}
     {targetBone {boneName=dyn_hair_m_primary_04}}
     {targetBone {boneName=dyn_hair_r_primary_01}}
     {targetBone {boneName=dyn_hair_r_primary_04}}
    ]
    constraintWeights [1,1,1,1,1,1,1,1,1,1]
   }
   n22 = "Constraint - dyn_hair_scd_q_master (PPPPPPPPP)"!(
    ComponentPose=@n21
    Alpha=1
    ConstraintWeights_0=0.0002
    ConstraintWeights_1:float=0.0079
    ConstraintWeights_2:float=0.244
    ConstraintWeights_3:float=0.1605
    ConstraintWeights_4:float=0.3376
    ConstraintWeights_5:float=0.2442
    ConstraintWeights_6:float=0.0001
    ConstraintWeights_7:float=0.0044
    ConstraintWeights_8:float=0.001
   ) -> (Pose) {
    boneToModify {boneName=dyn_hair_scd_q_master}
    constraintSetup [
     {targetBone {boneName=head}}
     {targetBone {boneName=dyn_hair_m_primary_01}}
     {targetBone {boneName=dyn_hair_m_primary_02}}
     {targetBone {boneName=dyn_hair_m_primary_03}}
     {targetBone {boneName=dyn_hair_r_primary_02}}
     {targetBone {boneName=dyn_hair_r_primary_03}}
     {targetBone {boneName=dyn_hair_m_primary_04}}
     {targetBone {boneName=dyn_hair_r_primary_01}}
     {targetBone {boneName=dyn_hair_r_primary_04}}
    ]
    constraintWeights [1,1,1,1,1,1,1,1,1]
   }
   n23 = "Constraint - dyn_hair_scd_r_master (PPPPPPPPP)"!(
    ComponentPose=@n22
    Alpha=1
    ConstraintWeights_0=0.0014
    ConstraintWeights_1:float=0
    ConstraintWeights_2:float=0.0035
    ConstraintWeights_3:float=0.0909
    ConstraintWeights_4:float=0.054
    ConstraintWeights_5:float=0.5015
    ConstraintWeights_6:float=0.335
    ConstraintWeights_7:float=0.0133
    ConstraintWeights_8:float=0.0004
   ) -> (Pose) {
    boneToModify {boneName=dyn_hair_scd_r_master}
    constraintSetup [
     {targetBone {boneName=head}}
     {targetBone {boneName=dyn_hair_l_primary_01}}
     {targetBone {boneName=dyn_hair_m_primary_01}}
     {targetBone {boneName=dyn_hair_m_primary_02}}
     {targetBone {boneName=dyn_hair_m_primary_03}}
     {targetBone {boneName=dyn_hair_r_primary_02}}
     {targetBone {boneName=dyn_hair_r_primary_03}}
     {targetBone {boneName=dyn_hair_r_primary_01}}
     {targetBone {boneName=dyn_hair_r_primary_04}}
    ]
    constraintWeights [1,1,1,1,1,1,1,1,1]
   }
   n24 = "Constraint - dyn_hair_scd_s_master (PPPPPPPP)"!(
    ComponentPose=@n23
    Alpha=1
    ConstraintWeights_0=0.0103
    ConstraintWeights_1:float=0.0002
    ConstraintWeights_2:float=0.015
    ConstraintWeights_3:float=0.0361
    ConstraintWeights_4:float=0.0055
    ConstraintWeights_5:float=0.7379
    ConstraintWeights_6:float=0.117
    ConstraintWeights_7:float=0.078
   ) -> (Pose) {
    boneToModify {boneName=dyn_hair_scd_s_master}
    constraintSetup [
     {targetBone {boneName=head}}
     {targetBone {boneName=dyn_hair_l_primary_01}}
     {targetBone {boneName=dyn_hair_m_primary_01}}
     {targetBone {boneName=dyn_hair_m_primary_02}}
     {targetBone {boneName=dyn_hair_m_primary_03}}
     {targetBone {boneName=dyn_hair_r_primary_02}}
     {targetBone {boneName=dyn_hair_r_primary_03}}
     {targetBone {boneName=dyn_hair_r_primary_01}}
    ]
    constraintWeights [1,1,1,1,1,1,1,1]
   }
   n25 = "Constraint - dyn_hair_scd_t_master (PPPPPPPP)"!(
    ComponentPose=@n24
    Alpha=1
    ConstraintWeights_0=0.039
    ConstraintWeights_1:float=0
    ConstraintWeights_2:float=0.0015
    ConstraintWeights_3:float=0.0009
    ConstraintWeights_4:float=0
    ConstraintWeights_5:float=0.7254
    ConstraintWeights_6:float=0.0255
    ConstraintWeights_7:float=0.2077
   ) -> (Pose) {
    boneToModify {boneName=dyn_hair_scd_t_master}
    constraintSetup [
     {targetBone {boneName=head}}
     {targetBone {boneName=dyn_hair_l_primary_01}}
     {targetBone {boneName=dyn_hair_m_primary_01}}
     {targetBone {boneName=dyn_hair_m_primary_02}}
     {targetBone {boneName=dyn_hair_m_primary_03}}
     {targetBone {boneName=dyn_hair_r_primary_02}}
     {targetBone {boneName=dyn_hair_r_primary_03}}
     {targetBone {boneName=dyn_hair_r_primary_01}}
    ]
    constraintWeights [1,1,1,1,1,1,1,1]
   }
   n26 = "Constraint - dyn_hair_scd_u_master (PPPPPPPP)"!(
    ComponentPose=@n25
    Alpha=1
    ConstraintWeights_0=0.209
    ConstraintWeights_1:float=0.0003
    ConstraintWeights_2:float=0.0065
    ConstraintWeights_3:float=0.0016
    ConstraintWeights_4:float=0.3164
    ConstraintWeights_5:float=0.0015
    ConstraintWeights_6:float=0
    ConstraintWeights_7:float=0.4646
   ) -> (Pose) {
    boneToModify {boneName=dyn_hair_scd_u_master}
    constraintSetup [
     {targetBone {boneName=head}}
     {targetBone {boneName=dyn_hair_l_primary_01}}
     {targetBone {boneName=dyn_hair_m_primary_01}}
     {targetBone {boneName=dyn_hair_m_primary_02}}
     {targetBone {boneName=dyn_hair_r_primary_02}}
     {targetBone {boneName=dyn_hair_r_primary_03}}
     {targetBone {boneName=dyn_hair_l_primary_02}}
     {targetBone {boneName=dyn_hair_r_primary_01}}
    ]
    constraintWeights [1,1,1,1,1,1,1,1]
   }
   n27 = "Constraint - dyn_hair_scd_v_master (PPPPPPPP)"!(
    ComponentPose=@n26
    Alpha=1
    ConstraintWeights_0=0.5314
    ConstraintWeights_1:float=0
    ConstraintWeights_2:float=0.0005
    ConstraintWeights_3:float=0.0002
    ConstraintWeights_4:float=0
    ConstraintWeights_5:float=0.2904
    ConstraintWeights_6:float=0.0282
    ConstraintWeights_7:float=0.1493
   ) -> (Pose) {
    boneToModify {boneName=dyn_hair_scd_v_master}
    constraintSetup [
     {targetBone {boneName=head}}
     {targetBone {boneName=dyn_hair_l_primary_01}}
     {targetBone {boneName=dyn_hair_m_primary_01}}
     {targetBone {boneName=dyn_hair_m_primary_02}}
     {targetBone {boneName=dyn_hair_m_primary_03}}
     {targetBone {boneName=dyn_hair_r_primary_02}}
     {targetBone {boneName=dyn_hair_r_primary_03}}
     {targetBone {boneName=dyn_hair_r_primary_01}}
    ]
    constraintWeights [1,1,1,1,1,1,1,1]
   }
   n28 = "Constraint - dyn_hair_scd_w_master (PPPPPPPPP)"!(
    ComponentPose=@n27
    Alpha=1
    ConstraintWeights_0=0.0043
    ConstraintWeights_1:float=0
    ConstraintWeights_2:float=0.0001
    ConstraintWeights_3:float=0.0003
    ConstraintWeights_4:float=0.0001
    ConstraintWeights_5:float=0.5096
    ConstraintWeights_6:float=0.4698
    ConstraintWeights_7:float=0.0158
    ConstraintWeights_8:float=0
   ) -> (Pose) {
    boneToModify {boneName=dyn_hair_scd_w_master}
    constraintSetup [
     {targetBone {boneName=head}}
     {targetBone {boneName=dyn_hair_l_primary_01}}
     {targetBone {boneName=dyn_hair_m_primary_01}}
     {targetBone {boneName=dyn_hair_m_primary_02}}
     {targetBone {boneName=dyn_hair_m_primary_03}}
     {targetBone {boneName=dyn_hair_r_primary_02}}
     {targetBone {boneName=dyn_hair_r_primary_03}}
     {targetBone {boneName=dyn_hair_r_primary_01}}
     {targetBone {boneName=dyn_hair_r_primary_04}}
    ]
    constraintWeights [1,1,1,1,1,1,1,1,1]
   }
   n29 = "Constraint - dyn_hair_scd_x_master (PPPPPPPP)"!(
    ComponentPose=@n28
    Alpha=1
    ConstraintWeights_0=0.0222
    ConstraintWeights_1:float=0
    ConstraintWeights_2:float=0.0004
    ConstraintWeights_3:float=0.0002
    ConstraintWeights_4:float=0.506
    ConstraintWeights_5:float=0.4528
    ConstraintWeights_6:float=0.0181
    ConstraintWeights_7:float=0.0003
   ) -> (Pose) {
    boneToModify {boneName=dyn_hair_scd_x_master}
    constraintSetup [
     {targetBone {boneName=head}}
     {targetBone {boneName=dyn_hair_m_primary_01}}
     {targetBone {boneName=dyn_hair_m_primary_02}}
     {targetBone {boneName=dyn_hair_m_primary_03}}
     {targetBone {boneName=dyn_hair_r_primary_02}}
     {targetBone {boneName=dyn_hair_r_primary_03}}
     {targetBone {boneName=dyn_hair_r_primary_01}}
     {targetBone {boneName=dyn_hair_r_primary_04}}
    ]
    constraintWeights [1,1,1,1,1,1,1,1]
   }
   n30 = "Constraint - dyn_hair_scd_y_master (PPPPPPPP)"!(
    ComponentPose=@n29
    Alpha=1
    ConstraintWeights_0=0
    ConstraintWeights_1:float=0.5838
    ConstraintWeights_2:float=0
    ConstraintWeights_3:float=0.0015
    ConstraintWeights_4:float=0.0303
    ConstraintWeights_5:float=0.0055
    ConstraintWeights_6:float=0.3582
    ConstraintWeights_7:float=0.0208
   ) -> (Pose) {
    boneToModify {boneName=dyn_hair_scd_y_master}
    constraintSetup [
     {targetBone {boneName=dyn_hair_l_primary_01}}
     {targetBone {boneName=dyn_hair_l_primary_03}}
     {targetBone {boneName=dyn_hair_m_primary_01}}
     {targetBone {boneName=dyn_hair_m_primary_02}}
     {targetBone {boneName=dyn_hair_m_primary_03}}
     {targetBone {boneName=dyn_hair_l_primary_02}}
     {targetBone {boneName=dyn_hair_l_primary_04}}
     {targetBone {boneName=dyn_hair_m_primary_04}}
    ]
    constraintWeights [1,1,1,1,1,1,1,1]
   }
   n31 = "Constraint - dyn_hair_scd_z_master (PPPPPPPP)"!(
    ComponentPose=@n30
    Alpha=1
    ConstraintWeights_0=0.001
    ConstraintWeights_1:float=0
    ConstraintWeights_2:float=0.0005
    ConstraintWeights_3:float=0
    ConstraintWeights_4:float=0
    ConstraintWeights_5:float=0.5141
    ConstraintWeights_6:float=0.4845
    ConstraintWeights_7:float=0
   ) -> (Pose) {
    boneToModify {boneName=dyn_hair_scd_z_master}
    constraintSetup [
     {targetBone {boneName=dyn_hair_l_primary_03}}
     {targetBone {boneName=dyn_hair_m_primary_02}}
     {targetBone {boneName=dyn_hair_m_primary_03}}
     {targetBone {boneName=dyn_hair_r_primary_03}}
     {targetBone {boneName=dyn_hair_l_primary_02}}
     {targetBone {boneName=dyn_hair_l_primary_04}}
     {targetBone {boneName=dyn_hair_m_primary_04}}
     {targetBone {boneName=dyn_hair_r_primary_04}}
    ]
    constraintWeights [1,1,1,1,1,1,1,1]
   }
   n32 = "Constraint - dyn_hair_scd_aa_master (PPPPPPP)"!(
    ComponentPose=@n31
    Alpha=1
    ConstraintWeights_0=0
    ConstraintWeights_1:float=0.0331
    ConstraintWeights_2:float=0
    ConstraintWeights_3:float=0.0013
    ConstraintWeights_4:float=0
    ConstraintWeights_5:float=0.9128
    ConstraintWeights_6:float=0.0528
   ) -> (Pose) {
    boneToModify {boneName=dyn_hair_scd_aa_master}
    constraintSetup [
     {targetBone {boneName=dyn_hair_l_primary_01}}
     {targetBone {boneName=dyn_hair_l_primary_03}}
     {targetBone {boneName=dyn_hair_m_primary_02}}
     {targetBone {boneName=dyn_hair_m_primary_03}}
     {targetBone {boneName=dyn_hair_l_primary_02}}
     {targetBone {boneName=dyn_hair_l_primary_04}}
     {targetBone {boneName=dyn_hair_m_primary_04}}
    ]
    constraintWeights [1,1,1,1,1,1,1]
   }
   n33 = "Constraint - dyn_hair_scd_ab_master (PPPPPPPP)"!(
    ComponentPose=@n32
    Alpha=1
    ConstraintWeights_0=0.0061
    ConstraintWeights_1:float=0
    ConstraintWeights_2:float=0.0504
    ConstraintWeights_3:float=0
    ConstraintWeights_4:float=0
    ConstraintWeights_5:float=0.1128
    ConstraintWeights_6:float=0.8306
    ConstraintWeights_7:float=0.0001
   ) -> (Pose) {
    boneToModify {boneName=dyn_hair_scd_ab_master}
    constraintSetup [
     {targetBone {boneName=dyn_hair_l_primary_03}}
     {targetBone {boneName=dyn_hair_m_primary_02}}
     {targetBone {boneName=dyn_hair_m_primary_03}}
     {targetBone {boneName=dyn_hair_r_primary_02}}
     {targetBone {boneName=dyn_hair_r_primary_03}}
     {targetBone {boneName=dyn_hair_l_primary_04}}
     {targetBone {boneName=dyn_hair_m_primary_04}}
     {targetBone {boneName=dyn_hair_r_primary_04}}
    ]
    constraintWeights [1,1,1,1,1,1,1,1]
   }
   n34 = "Constraint - dyn_hair_scd_ac_master (PPPPPPPPPP)"!(
    ComponentPose=@n33
    Alpha=1
    ConstraintWeights_0=0.0001
    ConstraintWeights_1:float=0
    ConstraintWeights_2:float=0.0001
    ConstraintWeights_3:float=0.0445
    ConstraintWeights_4:float=0
    ConstraintWeights_5:float=0.0037
    ConstraintWeights_6:float=0.0011
    ConstraintWeights_7:float=0.8931
    ConstraintWeights_8:float=0
    ConstraintWeights_9:float=0.0574
   ) -> (Pose) {
    boneToModify {boneName=dyn_hair_scd_ac_master}
    constraintSetup [
     {targetBone {boneName=dyn_hair_l_primary_03}}
     {targetBone {boneName=dyn_hair_m_primary_01}}
     {targetBone {boneName=dyn_hair_m_primary_02}}
     {targetBone {boneName=dyn_hair_m_primary_03}}
     {targetBone {boneName=dyn_hair_r_primary_02}}
     {targetBone {boneName=dyn_hair_r_primary_03}}
     {targetBone {boneName=dyn_hair_l_primary_04}}
     {targetBone {boneName=dyn_hair_m_primary_04}}
     {targetBone {boneName=dyn_hair_r_primary_01}}
     {targetBone {boneName=dyn_hair_r_primary_04}}
    ]
    constraintWeights [1,1,1,1,1,1,1,1,1,1]
   }
   n35 = "Constraint - dyn_hair_scd_ad_master (PPPPPPP)"!(
    ComponentPose=@n34
    Alpha=1
    ConstraintWeights_0=0
    ConstraintWeights_1:float=0
    ConstraintWeights_2:float=0
    ConstraintWeights_3:float=0
    ConstraintWeights_4:float=0
    ConstraintWeights_5:float=0.7994
    ConstraintWeights_6:float=0.2005
   ) -> (Pose) {
    boneToModify {boneName=dyn_hair_scd_ad_master}
    constraintSetup [
     {targetBone {boneName=dyn_hair_m_primary_02}}
     {targetBone {boneName=dyn_hair_m_primary_03}}
     {targetBone {boneName=dyn_hair_r_primary_02}}
     {targetBone {boneName=dyn_hair_r_primary_03}}
     {targetBone {boneName=dyn_hair_l_primary_04}}
     {targetBone {boneName=dyn_hair_m_primary_04}}
     {targetBone {boneName=dyn_hair_r_primary_04}}
    ]
    constraintWeights [1,1,1,1,1,1,1]
   }
   n36 = "Constraint - dyn_hair_scd_ae_master (PPPPPPPP)"!(
    ComponentPose=@n35
    Alpha=1
    ConstraintWeights_0=0
    ConstraintWeights_1:float=0
    ConstraintWeights_2:float=0.0006
    ConstraintWeights_3:float=0
    ConstraintWeights_4:float=0.0005
    ConstraintWeights_5:float=0
    ConstraintWeights_6:float=0.4174
    ConstraintWeights_7:float=0.5816
   ) -> (Pose) {
    boneToModify {boneName=dyn_hair_scd_ae_master}
    constraintSetup [
     {targetBone {boneName=dyn_hair_l_primary_03}}
     {targetBone {boneName=dyn_hair_m_primary_02}}
     {targetBone {boneName=dyn_hair_m_primary_03}}
     {targetBone {boneName=dyn_hair_r_primary_02}}
     {targetBone {boneName=dyn_hair_r_primary_03}}
     {targetBone {boneName=dyn_hair_l_primary_04}}
     {targetBone {boneName=dyn_hair_m_primary_04}}
     {targetBone {boneName=dyn_hair_r_primary_04}}
    ]
    constraintWeights [1,1,1,1,1,1,1,1]
   }
   n37 = "Constraint - dyn_hair_scd_af_master (PPPPPPPP)"!(
    ComponentPose=@n36
    Alpha=1
    ConstraintWeights_0=0
    ConstraintWeights_1:float=0
    ConstraintWeights_2:float=0.0019
    ConstraintWeights_3:float=0.0001
    ConstraintWeights_4:float=0.0464
    ConstraintWeights_5:float=0.0682
    ConstraintWeights_6:float=0
    ConstraintWeights_7:float=0.8834
   ) -> (Pose) {
    boneToModify {boneName=dyn_hair_scd_af_master}
    constraintSetup [
     {targetBone {boneName=head}}
     {targetBone {boneName=dyn_hair_m_primary_02}}
     {targetBone {boneName=dyn_hair_m_primary_03}}
     {targetBone {boneName=dyn_hair_r_primary_02}}
     {targetBone {boneName=dyn_hair_r_primary_03}}
     {targetBone {boneName=dyn_hair_m_primary_04}}
     {targetBone {boneName=dyn_hair_r_primary_01}}
     {targetBone {boneName=dyn_hair_r_primary_04}}
    ]
    constraintWeights [1,1,1,1,1,1,1,1]
   }
   n38 = "Constraint - dyn_hair_scd_ag_master (PPPPPPPP)"!(
    ComponentPose=@n37
    Alpha=1
    ConstraintWeights_0=0
    ConstraintWeights_1:float=0
    ConstraintWeights_2:float=0.0186
    ConstraintWeights_3:float=0.0003
    ConstraintWeights_4:float=0.1146
    ConstraintWeights_5:float=0.1111
    ConstraintWeights_6:float=0
    ConstraintWeights_7:float=0.7555
   ) -> (Pose) {
    boneToModify {boneName=dyn_hair_scd_ag_master}
    constraintSetup [
     {targetBone {boneName=head}}
     {targetBone {boneName=dyn_hair_m_primary_02}}
     {targetBone {boneName=dyn_hair_m_primary_03}}
     {targetBone {boneName=dyn_hair_r_primary_02}}
     {targetBone {boneName=dyn_hair_r_primary_03}}
     {targetBone {boneName=dyn_hair_m_primary_04}}
     {targetBone {boneName=dyn_hair_r_primary_01}}
     {targetBone {boneName=dyn_hair_r_primary_04}}
    ]
    constraintWeights [1,1,1,1,1,1,1,1]
   }
   n39 = "Constraint - dyn_hair_scd_ah_master (PPPPPPPP)"!(
    ComponentPose=@n38
    Alpha=1
    ConstraintWeights_0=0
    ConstraintWeights_1:float=0.0003
    ConstraintWeights_2:float=0.0314
    ConstraintWeights_3:float=0.008
    ConstraintWeights_4:float=0.4215
    ConstraintWeights_5:float=0.0216
    ConstraintWeights_6:float=0
    ConstraintWeights_7:float=0.5172
   ) -> (Pose) {
    boneToModify {boneName=dyn_hair_scd_ah_master}
    constraintSetup [
     {targetBone {boneName=dyn_hair_m_primary_01}}
     {targetBone {boneName=dyn_hair_m_primary_02}}
     {targetBone {boneName=dyn_hair_m_primary_03}}
     {targetBone {boneName=dyn_hair_r_primary_02}}
     {targetBone {boneName=dyn_hair_r_primary_03}}
     {targetBone {boneName=dyn_hair_m_primary_04}}
     {targetBone {boneName=dyn_hair_r_primary_01}}
     {targetBone {boneName=dyn_hair_r_primary_04}}
    ]
    constraintWeights [1,1,1,1,1,1,1,1]
   }
   n40 = "Constraint - dyn_hair_scd_ai_master (PPPPPPPP)"!(
    ComponentPose=@n39
    Alpha=1
    ConstraintWeights_0=0
    ConstraintWeights_1:float=0
    ConstraintWeights_2:float=0.0027
    ConstraintWeights_3:float=0.0028
    ConstraintWeights_4:float=0.2201
    ConstraintWeights_5:float=0.0044
    ConstraintWeights_6:float=0
    ConstraintWeights_7:float=0.77
   ) -> (Pose) {
    boneToModify {boneName=dyn_hair_scd_ai_master}
    constraintSetup [
     {targetBone {boneName=head}}
     {targetBone {boneName=dyn_hair_m_primary_02}}
     {targetBone {boneName=dyn_hair_m_primary_03}}
     {targetBone {boneName=dyn_hair_r_primary_02}}
     {targetBone {boneName=dyn_hair_r_primary_03}}
     {targetBone {boneName=dyn_hair_m_primary_04}}
     {targetBone {boneName=dyn_hair_r_primary_01}}
     {targetBone {boneName=dyn_hair_r_primary_04}}
    ]
    constraintWeights [1,1,1,1,1,1,1,1]
   }
   n41 = Anim Dynamics - Bone: dyn_hair_scd_a_01(ComponentPose=@n40, Alpha=1, ExternalForce="0,0,0") {
    angularDampingOverride=1
    angularSpringConstant=500
    bAngularSpring=true
    bOverrideAngularDamping=true
    bUseGravityOverride=true
    simulationSpace=BoneRelative
    boundBone {boneName=dyn_hair_scd_a_01}
    componentAppliedLinearAccClamp=<10000 10000 10000>
    componentLinearAccScale=<1 1 1>
    componentLinearVelScale=<1 1 1>
    gravityOverride {z=-980}
    physicsBodyDefinitions [{
      boundBone {boneName=dyn_hair_scd_a_01}
      boxExtents=<10 10 10>
      constraintSetup {angularLimitsMax=<3 10 35>, angularLimitsMin=<-3 -10 0>, angularTarget=<1 0 0>}
      localJointOffset=<0.3 0 0>
     }]
    relativeSpaceBone {boneName=root}
    simSpaceSettings {maxAngularAcceleration=350, maxAngularVelocity=50, simSpaceAngularAlpha=1}
   }
   n42 = Anim Dynamics - Bone: dyn_hair_scd_b_01(ComponentPose=@n41, Alpha=1, ExternalForce="0,0,0") {
    angularDampingOverride=1
    angularSpringConstant=500
    bAngularSpring=true
    bOverrideAngularDamping=true
    bUseGravityOverride=true
    simulationSpace=BoneRelative
    boundBone {boneName=dyn_hair_scd_b_01}
    componentAppliedLinearAccClamp=<10000 10000 10000>
    componentLinearAccScale=<1 1 1>
    componentLinearVelScale=<1 1 1>
    gravityOverride {z=-980}
    physicsBodyDefinitions [{
      boundBone {boneName=dyn_hair_scd_b_01}
      boxExtents=<10 10 10>
      constraintSetup {angularLimitsMax=<3 10 35>, angularLimitsMin=<-3 -10 0>, angularTarget=<1 0 0>}
      localJointOffset=<0.3 0 0>
     }]
    relativeSpaceBone {boneName=root}
    simSpaceSettings {maxAngularAcceleration=350, maxAngularVelocity=50, simSpaceAngularAlpha=1}
   }
   n43 = Anim Dynamics - Bone: dyn_hair_scd_c_01(ComponentPose=@n42, Alpha=1, ExternalForce="0,0,0") {
    angularDampingOverride=1
    angularSpringConstant=500
    bAngularSpring=true
    bOverrideAngularDamping=true
    bUseGravityOverride=true
    simulationSpace=BoneRelative
    boundBone {boneName=dyn_hair_scd_c_01}
    componentAppliedLinearAccClamp=<10000 10000 10000>
    componentLinearAccScale=<1 1 1>
    componentLinearVelScale=<1 1 1>
    gravityOverride {z=-980}
    physicsBodyDefinitions [{
      boundBone {boneName=dyn_hair_scd_c_01}
      boxExtents=<10 10 10>
      constraintSetup {angularLimitsMax=<3 10 35>, angularLimitsMin=<-3 -10 0>, angularTarget=<1 0 0>}
      localJointOffset=<0.3 0 0>
     }]
    relativeSpaceBone {boneName=root}
    simSpaceSettings {maxAngularAcceleration=350, maxAngularVelocity=50, simSpaceAngularAlpha=1}
   }
   n44 = Anim Dynamics - Bone: dyn_hair_scd_d_01(ComponentPose=@n43, Alpha=1, ExternalForce="0,0,0") {
    angularDampingOverride=1
    angularSpringConstant=500
    bAngularSpring=true
    bOverrideAngularDamping=true
    bUseGravityOverride=true
    simulationSpace=BoneRelative
    boundBone {boneName=dyn_hair_scd_d_01}
    componentAppliedLinearAccClamp=<10000 10000 10000>
    componentLinearAccScale=<1 1 1>
    componentLinearVelScale=<1 1 1>
    gravityOverride {z=-980}
    physicsBodyDefinitions [{
      boundBone {boneName=dyn_hair_scd_d_01}
      boxExtents=<10 10 10>
      constraintSetup {angularLimitsMax=<3 10 35>, angularLimitsMin=<-3 -10 0>, angularTarget=<1 0 0>}
      localJointOffset=<0.3 0 0>
     }]
    relativeSpaceBone {boneName=root}
    simSpaceSettings {maxAngularAcceleration=350, maxAngularVelocity=50, simSpaceAngularAlpha=1}
   }
   n45 = Anim Dynamics - Bone: dyn_hair_scd_e_01(ComponentPose=@n44, Alpha=1, ExternalForce="0,0,0") {
    angularDampingOverride=1
    angularSpringConstant=500
    bAngularSpring=true
    bOverrideAngularDamping=true
    bUseGravityOverride=true
    simulationSpace=BoneRelative
    boundBone {boneName=dyn_hair_scd_e_01}
    componentAppliedLinearAccClamp=<10000 10000 10000>
    componentLinearAccScale=<1 1 1>
    componentLinearVelScale=<1 1 1>
    gravityOverride {z=-980}
    physicsBodyDefinitions [{
      boundBone {boneName=dyn_hair_scd_e_01}
      boxExtents=<10 10 10>
      constraintSetup {angularLimitsMax=<3 10 35>, angularLimitsMin=<-3 -10 0>, angularTarget=<1 0 0>}
      localJointOffset=<0.1 0 0>
     }]
    relativeSpaceBone {boneName=root}
    simSpaceSettings {maxAngularAcceleration=350, maxAngularVelocity=50, simSpaceAngularAlpha=1}
   }
   n46 = Anim Dynamics - Bone: dyn_hair_scd_f_01(ComponentPose=@n45, Alpha=1, ExternalForce="0,0,0") {
    angularDampingOverride=1
    angularSpringConstant=500
    bAngularSpring=true
    bOverrideAngularDamping=true
    bUseGravityOverride=true
    simulationSpace=BoneRelative
    boundBone {boneName=dyn_hair_scd_f_01}
    componentAppliedLinearAccClamp=<10000 10000 10000>
    componentLinearAccScale=<1 1 1>
    componentLinearVelScale=<1 1 1>
    gravityOverride {z=-980}
    physicsBodyDefinitions [{
      boundBone {boneName=dyn_hair_scd_f_01}
      boxExtents=<10 10 10>
      constraintSetup {angularLimitsMax=<3 10 35>, angularLimitsMin=<-3 -10 0>, angularTarget=<1 0 0>}
      localJointOffset=<0.3 0 0>
     }]
    relativeSpaceBone {boneName=root}
    simSpaceSettings {maxAngularAcceleration=350, maxAngularVelocity=50, simSpaceAngularAlpha=1}
   }
   n47 = Anim Dynamics - Bone: dyn_hair_scd_g_01(ComponentPose=@n46, Alpha=1, ExternalForce="0,0,0") {
    angularDampingOverride=1
    angularSpringConstant=500
    bAngularSpring=true
    bOverrideAngularDamping=true
    bUseGravityOverride=true
    simulationSpace=BoneRelative
    boundBone {boneName=dyn_hair_scd_g_01}
    componentAppliedLinearAccClamp=<10000 10000 10000>
    componentLinearAccScale=<1 1 1>
    componentLinearVelScale=<1 1 1>
    gravityOverride {z=-980}
    physicsBodyDefinitions [{
      boundBone {boneName=dyn_hair_scd_g_01}
      boxExtents=<10 10 10>
      constraintSetup {angularLimitsMax=<3 10 35>, angularLimitsMin=<-3 -10 0>, angularTarget=<1 0 0>}
      localJointOffset=<0.1 0 0>
     }]
    relativeSpaceBone {boneName=root}
    simSpaceSettings {maxAngularAcceleration=350, maxAngularVelocity=50, simSpaceAngularAlpha=1}
   }
   n48 = Anim Dynamics - Bone: dyn_hair_scd_h_01(ComponentPose=@n47, Alpha=1, ExternalForce="0,0,0") {
    angularDampingOverride=1
    angularSpringConstant=500
    bAngularSpring=true
    bOverrideAngularDamping=true
    bUseGravityOverride=true
    simulationSpace=BoneRelative
    boundBone {boneName=dyn_hair_scd_h_01}
    componentAppliedLinearAccClamp=<10000 10000 10000>
    componentLinearAccScale=<1 1 1>
    componentLinearVelScale=<1 1 1>
    gravityOverride {z=-980}
    physicsBodyDefinitions [{
      boundBone {boneName=dyn_hair_scd_h_01}
      boxExtents=<10 10 10>
      constraintSetup {angularLimitsMax=<3 10 35>, angularLimitsMin=<-3 -10 0>, angularTarget=<1 0 0>}
      localJointOffset=<0.1 0 0>
     }]
    relativeSpaceBone {boneName=root}
    simSpaceSettings {maxAngularAcceleration=350, maxAngularVelocity=50, simSpaceAngularAlpha=1}
   }
   n49 = Anim Dynamics - Bone: dyn_hair_scd_i_01(ComponentPose=@n48, Alpha=1, ExternalForce="0,0,0") {
    angularDampingOverride=1
    angularSpringConstant=500
    bAngularSpring=true
    bOverrideAngularDamping=true
    bUseGravityOverride=true
    simulationSpace=BoneRelative
    boundBone {boneName=dyn_hair_scd_i_01}
    componentAppliedLinearAccClamp=<10000 10000 10000>
    componentLinearAccScale=<1 1 1>
    componentLinearVelScale=<1 1 1>
    gravityOverride {z=-980}
    physicsBodyDefinitions [{
      boundBone {boneName=dyn_hair_scd_i_01}
      boxExtents=<10 10 10>
      constraintSetup {angularLimitsMax=<3 10 35>, angularLimitsMin=<-3 -10 0>, angularTarget=<1 0 0>}
      localJointOffset=<0.1 0 0>
     }]
    relativeSpaceBone {boneName=root}
    simSpaceSettings {maxAngularAcceleration=350, maxAngularVelocity=50, simSpaceAngularAlpha=1}
   }
   n50 = Anim Dynamics - Bone: dyn_hair_scd_j_01(ComponentPose=@n49, Alpha=1, ExternalForce="0,0,0") {
    angularDampingOverride=1
    angularSpringConstant=500
    bAngularSpring=true
    bOverrideAngularDamping=true
    bUseGravityOverride=true
    simulationSpace=BoneRelative
    boundBone {boneName=dyn_hair_scd_j_01}
    componentAppliedLinearAccClamp=<10000 10000 10000>
    componentLinearAccScale=<1 1 1>
    componentLinearVelScale=<1 1 1>
    gravityOverride {z=-980}
    physicsBodyDefinitions [{
      boundBone {boneName=dyn_hair_scd_j_01}
      boxExtents=<10 10 10>
      constraintSetup {angularLimitsMax=<3 10 35>, angularLimitsMin=<-3 -10 0>, angularTarget=<1 0 0>}
      localJointOffset=<0.1 0 0>
     }]
    relativeSpaceBone {boneName=root}
    simSpaceSettings {maxAngularAcceleration=350, maxAngularVelocity=50, simSpaceAngularAlpha=1}
   }
   n51 = Anim Dynamics - Bone: dyn_hair_scd_k_01(ComponentPose=@n50, Alpha=1, ExternalForce="0,0,0") {
    angularDampingOverride=1
    angularSpringConstant=500
    bAngularSpring=true
    bOverrideAngularDamping=true
    bUseGravityOverride=true
    simulationSpace=BoneRelative
    boundBone {boneName=dyn_hair_scd_k_01}
    componentAppliedLinearAccClamp=<10000 10000 10000>
    componentLinearAccScale=<1 1 1>
    componentLinearVelScale=<1 1 1>
    gravityOverride {z=-980}
    physicsBodyDefinitions [{
      boundBone {boneName=dyn_hair_scd_k_01}
      boxExtents=<10 10 10>
      constraintSetup {angularLimitsMax=<3 10 35>, angularLimitsMin=<-3 -10 0>, angularTarget=<1 0 0>}
      localJointOffset=<0.1 0 0>
     }]
    relativeSpaceBone {boneName=root}
    simSpaceSettings {maxAngularAcceleration=350, maxAngularVelocity=50, simSpaceAngularAlpha=1}
   }
   n52 = Anim Dynamics - Bone: dyn_hair_scd_l_01(ComponentPose=@n51, Alpha=1, ExternalForce="0,0,0") {
    angularDampingOverride=1
    angularSpringConstant=500
    bAngularSpring=true
    bOverrideAngularDamping=true
    bUseGravityOverride=true
    simulationSpace=BoneRelative
    boundBone {boneName=dyn_hair_scd_l_01}
    componentAppliedLinearAccClamp=<10000 10000 10000>
    componentLinearAccScale=<1 1 1>
    componentLinearVelScale=<1 1 1>
    gravityOverride {z=-980}
    physicsBodyDefinitions [{
      boundBone {boneName=dyn_hair_scd_l_01}
      boxExtents=<10 10 10>
      constraintSetup {angularLimitsMax=<3 10 35>, angularLimitsMin=<-3 -10 0>, angularTarget=<1 0 0>}
      localJointOffset=<0.1 0 0>
     }]
    relativeSpaceBone {boneName=root}
    simSpaceSettings {maxAngularAcceleration=350, maxAngularVelocity=50, simSpaceAngularAlpha=1}
   }
   n53 = Anim Dynamics - Bone: dyn_hair_scd_m_01(ComponentPose=@n52, Alpha=1, ExternalForce="0,0,0") {
    angularDampingOverride=1
    angularSpringConstant=500
    bAngularSpring=true
    bOverrideAngularDamping=true
    bUseGravityOverride=true
    simulationSpace=BoneRelative
    boundBone {boneName=dyn_hair_scd_m_01}
    componentAppliedLinearAccClamp=<10000 10000 10000>
    componentLinearAccScale=<1 1 1>
    componentLinearVelScale=<1 1 1>
    gravityOverride {z=-980}
    physicsBodyDefinitions [{
      boundBone {boneName=dyn_hair_scd_m_01}
      boxExtents=<10 10 10>
      constraintSetup {angularLimitsMax=<3 10 35>, angularLimitsMin=<-3 -10 0>, angularTarget=<1 0 0>}
      localJointOffset=<0.1 0 0>
     }]
    relativeSpaceBone {boneName=root}
    simSpaceSettings {maxAngularAcceleration=350, maxAngularVelocity=50, simSpaceAngularAlpha=1}
   }
   n54 = Anim Dynamics - Bone: dyn_hair_scd_n_01(ComponentPose=@n53, Alpha=1, ExternalForce="0,0,0") {
    angularDampingOverride=1
    angularSpringConstant=500
    bAngularSpring=true
    bOverrideAngularDamping=true
    bUseGravityOverride=true
    simulationSpace=BoneRelative
    boundBone {boneName=dyn_hair_scd_n_01}
    componentAppliedLinearAccClamp=<10000 10000 10000>
    componentLinearAccScale=<1 1 1>
    componentLinearVelScale=<1 1 1>
    gravityOverride {z=-980}
    physicsBodyDefinitions [{
      boundBone {boneName=dyn_hair_scd_n_01}
      boxExtents=<10 10 10>
      constraintSetup {angularLimitsMax=<3 10 35>, angularLimitsMin=<-3 -10 0>, angularTarget=<1 0 0>}
      localJointOffset=<0.1 0 0>
     }]
    relativeSpaceBone {boneName=root}
    simSpaceSettings {maxAngularAcceleration=350, maxAngularVelocity=50, simSpaceAngularAlpha=1}
   }
   n55 = Anim Dynamics - Bone: dyn_hair_scd_o_01(ComponentPose=@n54, Alpha=1, ExternalForce="0,0,0") {
    angularDampingOverride=1
    angularSpringConstant=500
    bAngularSpring=true
    bOverrideAngularDamping=true
    bUseGravityOverride=true
    simulationSpace=BoneRelative
    boundBone {boneName=dyn_hair_scd_o_01}
    componentAppliedLinearAccClamp=<10000 10000 10000>
    componentLinearAccScale=<1 1 1>
    componentLinearVelScale=<1 1 1>
    gravityOverride {z=-980}
    physicsBodyDefinitions [{
      boundBone {boneName=dyn_hair_scd_o_01}
      boxExtents=<10 10 10>
      constraintSetup {angularLimitsMax=<3 10 35>, angularLimitsMin=<-3 -10 0>, angularTarget=<1 0 0>}
      localJointOffset=<0.1 0 0>
     }]
    relativeSpaceBone {boneName=root}
    simSpaceSettings {maxAngularAcceleration=350, maxAngularVelocity=50, simSpaceAngularAlpha=1}
   }
   n56 = Anim Dynamics - Bone: dyn_hair_scd_p_01(ComponentPose=@n55, Alpha=1, ExternalForce="0,0,0") {
    angularDampingOverride=1
    angularSpringConstant=500
    bAngularSpring=true
    bOverrideAngularDamping=true
    bUseGravityOverride=true
    simulationSpace=BoneRelative
    boundBone {boneName=dyn_hair_scd_p_01}
    componentAppliedLinearAccClamp=<10000 10000 10000>
    componentLinearAccScale=<1 1 1>
    componentLinearVelScale=<1 1 1>
    gravityOverride {z=-980}
    physicsBodyDefinitions [{
      boundBone {boneName=dyn_hair_scd_p_01}
      boxExtents=<10 10 10>
      constraintSetup {angularLimitsMax=<3 10 35>, angularLimitsMin=<-3 -10 0>, angularTarget=<1 0 0>}
      localJointOffset=<0.1 0 0>
     }]
    relativeSpaceBone {boneName=root}
    simSpaceSettings {maxAngularAcceleration=350, maxAngularVelocity=50, simSpaceAngularAlpha=1}
   }
   n57 = Anim Dynamics - Bone: dyn_hair_scd_q_01(ComponentPose=@n56, Alpha=1, ExternalForce="0,0,0") {
    angularDampingOverride=1
    angularSpringConstant=500
    bAngularSpring=true
    bOverrideAngularDamping=true
    bUseGravityOverride=true
    simulationSpace=BoneRelative
    boundBone {boneName=dyn_hair_scd_q_01}
    componentAppliedLinearAccClamp=<10000 10000 10000>
    componentLinearAccScale=<1 1 1>
    componentLinearVelScale=<1 1 1>
    gravityOverride {z=-980}
    physicsBodyDefinitions [{
      boundBone {boneName=dyn_hair_scd_q_01}
      boxExtents=<10 10 10>
      constraintSetup {angularLimitsMax=<3 10 35>, angularLimitsMin=<-3 -10 0>, angularTarget=<1 0 0>}
      localJointOffset=<0.1 0 0>
     }]
    relativeSpaceBone {boneName=root}
    simSpaceSettings {maxAngularAcceleration=350, maxAngularVelocity=50, simSpaceAngularAlpha=1}
   }
   n58 = Anim Dynamics - Bone: dyn_hair_scd_r_01(ComponentPose=@n57, Alpha=1, ExternalForce="0,0,0") {
    angularDampingOverride=1
    angularSpringConstant=500
    bAngularSpring=true
    bOverrideAngularDamping=true
    bUseGravityOverride=true
    simulationSpace=BoneRelative
    boundBone {boneName=dyn_hair_scd_r_01}
    componentAppliedLinearAccClamp=<10000 10000 10000>
    componentLinearAccScale=<1 1 1>
    componentLinearVelScale=<1 1 1>
    gravityOverride {z=-980}
    physicsBodyDefinitions [{
      boundBone {boneName=dyn_hair_scd_r_01}
      boxExtents=<10 10 10>
      constraintSetup {angularLimitsMax=<3 10 35>, angularLimitsMin=<-3 -10 0>, angularTarget=<1 0 0>}
      localJointOffset=<0.1 0 0>
     }]
    relativeSpaceBone {boneName=root}
    simSpaceSettings {maxAngularAcceleration=350, maxAngularVelocity=50, simSpaceAngularAlpha=1}
   }
   n59 = Anim Dynamics - Bone: dyn_hair_scd_s_01(ComponentPose=@n58, Alpha=1, ExternalForce="0,0,0") {
    angularDampingOverride=1
    angularSpringConstant=500
    bAngularSpring=true
    bOverrideAngularDamping=true
    bUseGravityOverride=true
    simulationSpace=BoneRelative
    boundBone {boneName=dyn_hair_scd_s_01}
    componentAppliedLinearAccClamp=<10000 10000 10000>
    componentLinearAccScale=<1 1 1>
    componentLinearVelScale=<1 1 1>
    gravityOverride {z=-980}
    physicsBodyDefinitions [{
      boundBone {boneName=dyn_hair_scd_s_01}
      boxExtents=<10 10 10>
      constraintSetup {angularLimitsMax=<3 10 35>, angularLimitsMin=<-3 -10 0>, angularTarget=<1 0 0>}
      localJointOffset=<0.1 0 0>
     }]
    relativeSpaceBone {boneName=root}
    simSpaceSettings {maxAngularAcceleration=350, maxAngularVelocity=50, simSpaceAngularAlpha=1}
   }
   n60 = Anim Dynamics - Bone: dyn_hair_scd_t_01(ComponentPose=@n59, Alpha=1, ExternalForce="0,0,0") {
    angularDampingOverride=1
    angularSpringConstant=500
    bAngularSpring=true
    bOverrideAngularDamping=true
    bUseGravityOverride=true
    simulationSpace=BoneRelative
    boundBone {boneName=dyn_hair_scd_t_01}
    componentAppliedLinearAccClamp=<10000 10000 10000>
    componentLinearAccScale=<1 1 1>
    componentLinearVelScale=<1 1 1>
    gravityOverride {z=-980}
    physicsBodyDefinitions [{
      boundBone {boneName=dyn_hair_scd_t_01}
      boxExtents=<10 10 10>
      constraintSetup {angularLimitsMax=<3 10 35>, angularLimitsMin=<-3 -10 0>, angularTarget=<1 0 0>}
      localJointOffset=<0.1 0 0>
     }]
    relativeSpaceBone {boneName=root}
    simSpaceSettings {maxAngularAcceleration=350, maxAngularVelocity=50, simSpaceAngularAlpha=1}
   }
   n61 = Anim Dynamics - Bone: dyn_hair_scd_u_01(ComponentPose=@n60, Alpha=1, ExternalForce="0,0,0") {
    angularDampingOverride=1
    angularSpringConstant=500
    bAngularSpring=true
    bOverrideAngularDamping=true
    bUseGravityOverride=true
    simulationSpace=BoneRelative
    boundBone {boneName=dyn_hair_scd_u_01}
    componentAppliedLinearAccClamp=<10000 10000 10000>
    componentLinearAccScale=<1 1 1>
    componentLinearVelScale=<1 1 1>
    gravityOverride {z=-980}
    physicsBodyDefinitions [{
      boundBone {boneName=dyn_hair_scd_u_01}
      boxExtents=<10 10 10>
      constraintSetup {angularLimitsMax=<3 10 35>, angularLimitsMin=<-3 -10 0>, angularTarget=<1 0 0>}
      localJointOffset=<0.1 0 0>
     }]
    relativeSpaceBone {boneName=root}
    simSpaceSettings {maxAngularAcceleration=350, maxAngularVelocity=50, simSpaceAngularAlpha=1}
   }
   n62 = Anim Dynamics - Bone: dyn_hair_scd_v_01(ComponentPose=@n61, Alpha=1, ExternalForce="0,0,0") {
    angularDampingOverride=1
    angularSpringConstant=500
    bAngularSpring=true
    bOverrideAngularDamping=true
    bUseGravityOverride=true
    simulationSpace=BoneRelative
    boundBone {boneName=dyn_hair_scd_v_01}
    componentAppliedLinearAccClamp=<10000 10000 10000>
    componentLinearAccScale=<1 1 1>
    componentLinearVelScale=<1 1 1>
    gravityOverride {z=-980}
    physicsBodyDefinitions [{
      boundBone {boneName=dyn_hair_scd_v_01}
      boxExtents=<10 10 10>
      constraintSetup {angularLimitsMax=<3 10 35>, angularLimitsMin=<-3 -10 0>, angularTarget=<1 0 0>}
      localJointOffset=<0.1 0 0>
     }]
    relativeSpaceBone {boneName=root}
    simSpaceSettings {maxAngularAcceleration=350, maxAngularVelocity=50, simSpaceAngularAlpha=1}
   }
   n63 = Anim Dynamics - Bone: dyn_hair_scd_w_01(ComponentPose=@n62, Alpha=1, ExternalForce="0,0,0") {
    angularDampingOverride=1
    angularSpringConstant=500
    bAngularSpring=true
    bOverrideAngularDamping=true
    bUseGravityOverride=true
    simulationSpace=BoneRelative
    boundBone {boneName=dyn_hair_scd_w_01}
    componentAppliedLinearAccClamp=<10000 10000 10000>
    componentLinearAccScale=<1 1 1>
    componentLinearVelScale=<1 1 1>
    gravityOverride {z=-980}
    physicsBodyDefinitions [{
      boundBone {boneName=dyn_hair_scd_w_01}
      boxExtents=<10 10 10>
      constraintSetup {angularLimitsMax=<3 10 35>, angularLimitsMin=<-3 -10 0>, angularTarget=<1 0 0>}
      localJointOffset=<0.1 0 0>
     }]
    relativeSpaceBone {boneName=root}
    simSpaceSettings {maxAngularAcceleration=350, maxAngularVelocity=50, simSpaceAngularAlpha=1}
   }
   n64 = Anim Dynamics - Bone: dyn_hair_scd_x_01(ComponentPose=@n63, Alpha=1, ExternalForce="0,0,0") {
    angularDampingOverride=1
    angularSpringConstant=500
    bAngularSpring=true
    bOverrideAngularDamping=true
    bUseGravityOverride=true
    simulationSpace=BoneRelative
    boundBone {boneName=dyn_hair_scd_x_01}
    componentAppliedLinearAccClamp=<10000 10000 10000>
    componentLinearAccScale=<1 1 1>
    componentLinearVelScale=<1 1 1>
    gravityOverride {z=-980}
    physicsBodyDefinitions [{
      boundBone {boneName=dyn_hair_scd_x_01}
      boxExtents=<10 10 10>
      constraintSetup {angularLimitsMax=<3 10 35>, angularLimitsMin=<-3 -10 0>, angularTarget=<1 0 0>}
      localJointOffset=<0.1 0 0>
     }]
    relativeSpaceBone {boneName=root}
    simSpaceSettings {maxAngularAcceleration=350, maxAngularVelocity=50, simSpaceAngularAlpha=1}
   }
   n65 = Anim Dynamics - Bone: dyn_hair_scd_y_01(ComponentPose=@n64, Alpha=1, ExternalForce="0,0,0") {
    angularDampingOverride=1
    angularSpringConstant=500
    bAngularSpring=true
    bOverrideAngularDamping=true
    bUseGravityOverride=true
    simulationSpace=BoneRelative
    boundBone {boneName=dyn_hair_scd_y_01}
    componentAppliedLinearAccClamp=<10000 10000 10000>
    componentLinearAccScale=<1 1 1>
    componentLinearVelScale=<1 1 1>
    gravityOverride {z=-980}
    physicsBodyDefinitions [{
      boundBone {boneName=dyn_hair_scd_y_01}
      boxExtents=<10 10 10>
      constraintSetup {angularLimitsMax=<3 10 35>, angularLimitsMin=<-3 -10 0>, angularTarget=<1 0 0>}
      localJointOffset=<0.1 0 0>
     }]
    relativeSpaceBone {boneName=root}
    simSpaceSettings {maxAngularAcceleration=350, maxAngularVelocity=50, simSpaceAngularAlpha=1}
   }
   n66 = Anim Dynamics - Bone: dyn_hair_scd_z_01(ComponentPose=@n65, Alpha=1, ExternalForce="0,0,0") {
    angularDampingOverride=1
    angularSpringConstant=500
    bAngularSpring=true
    bOverrideAngularDamping=true
    bUseGravityOverride=true
    simulationSpace=BoneRelative
    boundBone {boneName=dyn_hair_scd_z_01}
    componentAppliedLinearAccClamp=<10000 10000 10000>
    componentLinearAccScale=<1 1 1>
    componentLinearVelScale=<1 1 1>
    gravityOverride {z=-980}
    physicsBodyDefinitions [{
      boundBone {boneName=dyn_hair_scd_z_01}
      boxExtents=<10 10 10>
      constraintSetup {angularLimitsMax=<3 10 35>, angularLimitsMin=<-3 -10 0>, angularTarget=<1 0 0>}
      localJointOffset=<0.1 0 0>
     }]
    relativeSpaceBone {boneName=root}
    simSpaceSettings {maxAngularAcceleration=350, maxAngularVelocity=50, simSpaceAngularAlpha=1}
   }
   n67 = Anim Dynamics - Bone: dyn_hair_scd_aa_01(ComponentPose=@n66, Alpha=1, ExternalForce="0,0,0") {
    angularDampingOverride=1
    angularSpringConstant=500
    bAngularSpring=true
    bOverrideAngularDamping=true
    bUseGravityOverride=true
    simulationSpace=BoneRelative
    boundBone {boneName=dyn_hair_scd_aa_01}
    componentAppliedLinearAccClamp=<10000 10000 10000>
    componentLinearAccScale=<1 1 1>
    componentLinearVelScale=<1 1 1>
    gravityOverride {z=-980}
    physicsBodyDefinitions [{
      boundBone {boneName=dyn_hair_scd_aa_01}
      boxExtents=<10 10 10>
      constraintSetup {angularLimitsMax=<3 10 35>, angularLimitsMin=<-3 -10 0>, angularTarget=<1 0 0>}
      localJointOffset=<0.1 0 0>
     }]
    relativeSpaceBone {boneName=root}
    simSpaceSettings {maxAngularAcceleration=350, maxAngularVelocity=50, simSpaceAngularAlpha=1}
   }
   n68 = Anim Dynamics - Bone: dyn_hair_scd_ab_01(ComponentPose=@n67, Alpha=1, ExternalForce="0,0,0") {
    angularDampingOverride=1
    angularSpringConstant=500
    bAngularSpring=true
    bOverrideAngularDamping=true
    bUseGravityOverride=true
    simulationSpace=BoneRelative
    boundBone {boneName=dyn_hair_scd_ab_01}
    componentAppliedLinearAccClamp=<10000 10000 10000>
    componentLinearAccScale=<1 1 1>
    componentLinearVelScale=<1 1 1>
    gravityOverride {z=-980}
    physicsBodyDefinitions [{
      boundBone {boneName=dyn_hair_scd_ab_01}
      boxExtents=<10 10 10>
      constraintSetup {angularLimitsMax=<3 10 35>, angularLimitsMin=<-3 -10 0>, angularTarget=<1 0 0>}
      localJointOffset=<0.1 0 0>
     }]
    relativeSpaceBone {boneName=root}
    simSpaceSettings {maxAngularAcceleration=350, maxAngularVelocity=50, simSpaceAngularAlpha=1}
   }
   n69 = Anim Dynamics - Bone: dyn_hair_scd_ac_01(ComponentPose=@n68, Alpha=1, ExternalForce="0,0,0") {
    angularDampingOverride=1
    angularSpringConstant=500
    bAngularSpring=true
    bOverrideAngularDamping=true
    bUseGravityOverride=true
    simulationSpace=BoneRelative
    boundBone {boneName=dyn_hair_scd_ac_01}
    componentAppliedLinearAccClamp=<10000 10000 10000>
    componentLinearAccScale=<1 1 1>
    componentLinearVelScale=<1 1 1>
    gravityOverride {z=-980}
    physicsBodyDefinitions [{
      boundBone {boneName=dyn_hair_scd_ac_01}
      boxExtents=<10 10 10>
      constraintSetup {angularLimitsMax=<3 10 35>, angularLimitsMin=<-3 -10 0>, angularTarget=<1 0 0>}
      localJointOffset=<0.1 0 0>
     }]
    relativeSpaceBone {boneName=root}
    simSpaceSettings {maxAngularAcceleration=350, maxAngularVelocity=50, simSpaceAngularAlpha=1}
   }
   n70 = Anim Dynamics - Bone: dyn_hair_scd_ad_01(ComponentPose=@n69, Alpha=1, ExternalForce="0,0,0") {
    angularDampingOverride=1
    angularSpringConstant=500
    bAngularSpring=true
    bOverrideAngularDamping=true
    bUseGravityOverride=true
    simulationSpace=BoneRelative
    boundBone {boneName=dyn_hair_scd_ad_01}
    componentAppliedLinearAccClamp=<10000 10000 10000>
    componentLinearAccScale=<1 1 1>
    componentLinearVelScale=<1 1 1>
    gravityOverride {z=-980}
    physicsBodyDefinitions [{
      boundBone {boneName=dyn_hair_scd_ad_01}
      boxExtents=<10 10 10>
      constraintSetup {angularLimitsMax=<3 10 35>, angularLimitsMin=<-3 -10 0>, angularTarget=<1 0 0>}
      localJointOffset=<0.1 0 0>
     }]
    relativeSpaceBone {boneName=root}
    simSpaceSettings {maxAngularAcceleration=350, maxAngularVelocity=50, simSpaceAngularAlpha=1}
   }
   n71 = Anim Dynamics - Bone: dyn_hair_scd_ae_01(ComponentPose=@n70, Alpha=1, ExternalForce="0,0,0") {
    angularDampingOverride=1
    angularSpringConstant=500
    bAngularSpring=true
    bOverrideAngularDamping=true
    bUseGravityOverride=true
    simulationSpace=BoneRelative
    boundBone {boneName=dyn_hair_scd_ae_01}
    componentAppliedLinearAccClamp=<10000 10000 10000>
    componentLinearAccScale=<1 1 1>
    componentLinearVelScale=<1 1 1>
    gravityOverride {z=-980}
    physicsBodyDefinitions [{
      boundBone {boneName=dyn_hair_scd_ae_01}
      boxExtents=<10 10 10>
      constraintSetup {angularLimitsMax=<3 10 35>, angularLimitsMin=<-3 -10 0>, angularTarget=<1 0 0>}
      localJointOffset=<0.1 0 0>
     }]
    relativeSpaceBone {boneName=root}
    simSpaceSettings {maxAngularAcceleration=350, maxAngularVelocity=50, simSpaceAngularAlpha=1}
   }
   n72 = Anim Dynamics - Bone: dyn_hair_scd_af_01(ComponentPose=@n71, Alpha=1, ExternalForce="0,0,0") {
    angularDampingOverride=1
    angularSpringConstant=500
    bAngularSpring=true
    bOverrideAngularDamping=true
    bUseGravityOverride=true
    simulationSpace=BoneRelative
    boundBone {boneName=dyn_hair_scd_af_01}
    componentAppliedLinearAccClamp=<10000 10000 10000>
    componentLinearAccScale=<1 1 1>
    componentLinearVelScale=<1 1 1>
    gravityOverride {z=-980}
    physicsBodyDefinitions [{
      boundBone {boneName=dyn_hair_scd_af_01}
      boxExtents=<10 10 10>
      constraintSetup {angularLimitsMax=<3 10 35>, angularLimitsMin=<-3 -10 0>, angularTarget=<1 0 0>}
      localJointOffset=<0.1 0 0>
     }]
    relativeSpaceBone {boneName=root}
    simSpaceSettings {maxAngularAcceleration=350, maxAngularVelocity=50, simSpaceAngularAlpha=1}
   }
   n73 = Anim Dynamics - Bone: dyn_hair_scd_ag_01(ComponentPose=@n72, Alpha=1, ExternalForce="0,0,0") {
    angularDampingOverride=1
    angularSpringConstant=500
    bAngularSpring=true
    bOverrideAngularDamping=true
    bUseGravityOverride=true
    simulationSpace=BoneRelative
    boundBone {boneName=dyn_hair_scd_ag_01}
    componentAppliedLinearAccClamp=<10000 10000 10000>
    componentLinearAccScale=<1 1 1>
    componentLinearVelScale=<1 1 1>
    gravityOverride {z=-980}
    physicsBodyDefinitions [{
      boundBone {boneName=dyn_hair_scd_ag_01}
      boxExtents=<10 10 10>
      constraintSetup {angularLimitsMax=<3 10 35>, angularLimitsMin=<-3 -10 0>, angularTarget=<1 0 0>}
      localJointOffset=<0.1 0 0>
     }]
    relativeSpaceBone {boneName=root}
    simSpaceSettings {maxAngularAcceleration=350, maxAngularVelocity=50, simSpaceAngularAlpha=1}
   }
   n74 = Anim Dynamics - Bone: dyn_hair_scd_ah_01(ComponentPose=@n73, Alpha=1, ExternalForce="0,0,0") {
    angularDampingOverride=1
    angularSpringConstant=500
    bAngularSpring=true
    bOverrideAngularDamping=true
    bUseGravityOverride=true
    simulationSpace=BoneRelative
    boundBone {boneName=dyn_hair_scd_ah_01}
    componentAppliedLinearAccClamp=<10000 10000 10000>
    componentLinearAccScale=<1 1 1>
    componentLinearVelScale=<1 1 1>
    gravityOverride {z=-980}
    physicsBodyDefinitions [{
      boundBone {boneName=dyn_hair_scd_ah_01}
      boxExtents=<10 10 10>
      constraintSetup {angularLimitsMax=<3 10 35>, angularLimitsMin=<-3 -10 0>, angularTarget=<1 0 0>}
      localJointOffset=<0.1 0 0>
     }]
    relativeSpaceBone {boneName=root}
    simSpaceSettings {maxAngularAcceleration=350, maxAngularVelocity=50, simSpaceAngularAlpha=1}
   }
   n75 = Anim Dynamics - Bone: dyn_hair_scd_ai_01(ComponentPose=@n74, Alpha=1, ExternalForce="0,0,0") {
    angularDampingOverride=1
    angularSpringConstant=500
    bAngularSpring=true
    bOverrideAngularDamping=true
    bUseGravityOverride=true
    simulationSpace=BoneRelative
    boundBone {boneName=dyn_hair_scd_ai_01}
    componentAppliedLinearAccClamp=<10000 10000 10000>
    componentLinearAccScale=<1 1 1>
    componentLinearVelScale=<1 1 1>
    gravityOverride {z=-980}
    physicsBodyDefinitions [{
      boundBone {boneName=dyn_hair_scd_ai_01}
      boxExtents=<10 10 10>
      constraintSetup {angularLimitsMax=<3 10 35>, angularLimitsMin=<-3 -10 0>, angularTarget=<1 0 0>}
      localJointOffset=<0.1 0 0>
     }]
    relativeSpaceBone {boneName=root}
    simSpaceSettings {maxAngularAcceleration=350, maxAngularVelocity=50, simSpaceAngularAlpha=1}
   }
   Output Pose(Result=Component To Local(ComponentPose=@n75)) {layerGroup=None}
   n76 = Control Rig!(Source=Copy Pose From Mesh() {bUseAttachedParent=true}, Alpha, isWearingBackbling:bool=Get isWearingBackbling()) -> (Pose) {controlRigAssetReference {controlRigAsset=&/M_MED_VealCadet_FaceAcc_CtrlRig}}
   Save cached pose 'PreSim'(Pose=@n76) {cacheName=PreSim}
  ]}
 EventGraph {nodes [
   Event BlueprintUpdateAnimation()
   > Sequence()
   > n1 = "Gravity Override State Function (Parameter Asset) (Nativized)"(
    CharacterPartAnimInstance=Self-Reference()
    GameTimeInSeconds=GetGameTimeInSeconds()
    Parameters=&/M_MED_VealCadet_FaceAcc_Parameters
   )
   > Set Gravity Override(Gravity Override=@n1)
   Event OnCharacterCustomizationLoaded()
   n2 = CheckForBackbling()
   > Set isWearingBackbling(isWearingBackbling=OR Boolean!(A=@n2, B=@n2.bIsBackblingACape, C:bool=@n2.bIsBackblingLong) -> (ReturnValue))
   > Set is_WearingCape(is_WearingCape:bool=@n2.bIsBackblingLong) -> (Output_Get:bool)
   Branch(Condition=NOT Boolean(A=Get bIsPlayingEmote()))
   n3 = BreakTransform(InTransform=GetDeltaTransformFromRefPose(self=GetOwnerSkeletalMeshComponent(), BoneName=head, BaseName=neck_02))
   > Set HeadLimitAlpha(HeadLimitAlpha:double=MapRangeClamped(Value=BreakRotator(InRot=@n3.Rotation).Yaw, InRangeA=-10, InRangeB=-70)) -> (Output_Get:double)
  ]}
]
variables [
 Gravity Override {type=Vector}
 HeadLimitAlpha {type=number}
 is_WearingCape {type=boolean}
 isWearingBackbling {type=boolean}
]
